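# Control Rig for the F_MED_BiscuitFluff character body that applies corrective bone offsets to its skirt, sleeve, back and thigh helper bones as the body poses.
# A shared deform rig function runs first, then parent constraints blend helpers between thighs and pelvis.
# Spherical pose readers, pelvis height, knee distance and Kalman-smoothed movement inputs weight layered Modify Transforms nodes.
# Flags for emote, alt skin, skydive and floating branch the corrections.
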ControlRigBlueprint /BRCosmetics/Characters/Player/Female/Medium/Bodies/F_MED_BiscuitFluff/Meshes/F_MED_BiscuitFluff_CtrlRig
graphs [
 Multiply_1_1_1_1_1_3_3_ContainedGraph {nodes [
   n1 = RigVMFunctionEntryNode()
   RigVMFunctionReturnNode(Result=&RigVMFunction_MathDoubleMul(B=@n1.C, A=&RigVMFunction_MathDoubleMul(A=@n1.A, B=@n1.B).Result).Result)
  ]}
 Multiply_1_1_2_1_1_ContainedGraph {nodes [
   n1 = RigVMFunctionEntryNode()
   n2 = &RigVMFunction_MathDoubleMul(A=@n1.A, B=@n1.B)
   RigVMFunctionReturnNode(Result=&RigVMFunction_MathDoubleMul(B=@n1.D, A=&RigVMFunction_MathDoubleMul(B=@n1.C, A=@n2.Result).Result).Result)
  ]}
 Multiply_1_1_2_2_ContainedGraph {nodes [
   n1 = RigVMFunctionEntryNode()
   RigVMFunctionReturnNode(Result=&RigVMFunction_MathDoubleMul(B=@n1.C, A=&RigVMFunction_MathDoubleMul(A=@n1.A, B=@n1.B).Result).Result)
  ]}
 Multiply_1_1_3_ContainedGraph {nodes [
   n1 = RigVMFunctionEntryNode()
   RigVMFunctionReturnNode(Result=&RigVMFunction_MathDoubleMul(B=@n1.C, A=&RigVMFunction_MathDoubleMul(A=@n1.A, B=@n1.B).Result).Result)
  ]}
 Multiply_1_1_ContainedGraph {nodes [
   n1 = RigVMFunctionEntryNode()
   RigVMFunctionReturnNode(Result=&RigVMFunction_MathDoubleMul(B=@n1.C, A=&RigVMFunction_MathDoubleMul(A=@n1.A, B=@n1.B).Result).Result)
  ]}
 Multiply_2_1_ContainedGraph {nodes [
   n1 = RigVMFunctionEntryNode()
   n2 = &RigVMFunction_MathDoubleMul(A=@n1.A, B=@n1.B)
   RigVMFunctionReturnNode(Result=&RigVMFunction_MathDoubleMul(B=@n1.D, A=&RigVMFunction_MathDoubleMul(B=@n1.C, A=@n2.Result).Result).Result)
  ]}
 Multiply_2_1_ContainedGraph {nodes [
   n1 = RigVMFunctionEntryNode()
   n2 = &RigVMFunction_MathDoubleMul(A=@n1.A, B=@n1.B)
   RigVMFunctionReturnNode(Result=&RigVMFunction_MathDoubleMul(B=@n1.D, A=&RigVMFunction_MathDoubleMul(B=@n1.C, A=@n2.Result).Result).Result)
  ]}
 Multiply_2_1_ContainedGraph {nodes [
   n1 = RigVMFunctionEntryNode()
   n2 = &RigVMFunction_MathDoubleMul(A=@n1.A, B=@n1.B)
   RigVMFunctionReturnNode(Result=&RigVMFunction_MathDoubleMul(B=@n1.D, A=&RigVMFunction_MathDoubleMul(B=@n1.C, A=@n2.Result).Result).Result)
  ]}
 RigVMFunctionLibrary {links [], nodes []}
 RigVMFunction_Sequence_1_1_ContainedGraph {nodes [
   n1 = RigVMFunctionEntryNode()
   n2 = &RigVMFunction_Sequence(ExecuteContext=@n1.ExecuteContext)
   n3 = &RigVMFunction_Sequence(ExecuteContext=@n2.B)
   n4 = &RigVMFunction_Sequence(ExecuteContext=@n3.B)
   n5 = &RigVMFunction_Sequence(ExecuteContext=@n4.B)
   RigVMFunctionReturnNode(A=@n2.A, B=@n3.A, C=@n4.A, D=@n5.A, E=@n5.B)
  ]}
 RigVMFunction_Sequence_2_1_1_ContainedGraph {nodes [
   n1 = RigVMFunctionEntryNode()
   n2 = &RigVMFunction_Sequence(ExecuteContext=@n1.ExecuteContext)
   n3 = &RigVMFunction_Sequence(ExecuteContext=@n2.B)
   n4 = &RigVMFunction_Sequence(ExecuteContext=@n3.B)
   n5 = &RigVMFunction_Sequence(ExecuteContext=@n4.B)
   n6 = &RigVMFunction_Sequence(ExecuteContext=@n5.B)
   n7 = &RigVMFunction_Sequence(ExecuteContext=@n6.B)
   n8 = &RigVMFunction_Sequence(ExecuteContext=@n7.B)
   n9 = &RigVMFunction_Sequence(ExecuteContext=@n8.B)
   n10 = &RigVMFunction_Sequence(ExecuteContext=@n9.B)
   RigVMFunctionReturnNode(A=@n2.A, B=@n3.A, C=@n4.A, D=@n5.A, E=@n6.A, F=@n7.A, G=@n8.A, H=@n9.A, I=@n10.A, J=@n10.B)
  ]}
 RigVMModel {nodes [
   Get Delta Ref_2_1 = RigVMFunctionReferenceNode(ExecuteContext=(), Item=(Type=Bone,Name="thigh_l"))
   > &RigUnit_ModifyTransforms(
    ItemToModify=((Item=(Type=Bone,Name="base_sleeve_l"),Transform=(Rotation=(X=0,Y=-0.1543,Z=0,W=0.988),Translation=(X=0,Y=0,Z=3),Scale3D=(X=1,Y=1,Z=1))))
    WeightMinimum=0
    Weight=&RigVMFunction_MathDoubleRemap(SourceMaximum=15, SourceMinimum=4, TargetMaximum=1, TargetMinimum=0, Value=@"Get Delta Ref_2_1".Result.Rotation.Roll).Result
   )
   &RigUnit_ParentConstraint(
    AdvancedSettings=(RotationOrderForFilter=XZY)
    Child=(Type=Bone,Name="dyn_skirt_bk_l_mstr")
    ExecutePin=()
    Filter=(TranslationFilter=(bX=true,bY=true,bZ=true),RotationFilter=(bX=true,bY=true,bZ=true),ScaleFilter=(bX=false,bY=false,bZ=false))
    Parents=((Item=(Type=Bone,Name="thigh_r"),Weight=0.3),(Item=(Type=Bone,Name="pelvis"),Weight=0.7))
    Weight=1
    bMaintainOffset=true
   )
   > &RigUnit_ParentConstraint(
    AdvancedSettings=(RotationOrderForFilter=XZY)
    Child=(Type=Bone,Name="Base_thingCorr_l")
    Filter=(TranslationFilter=(bX=true,bY=true,bZ=true),RotationFilter=(bX=true,bY=true,bZ=true),ScaleFilter=(bX=false,bY=false,bZ=false))
    Parents=((Item=(Type=Bone,Name="thigh_l"),Weight=0.3),(Item=(Type=Bone,Name="pelvis"),Weight=0.7))
    Weight=1
    bMaintainOffset=true
   )
   > &RigUnit_ParentConstraint(
    AdvancedSettings=(RotationOrderForFilter=XZY)
    Child=(Type=Bone,Name="dyn_skirt_bk_mstr")
    Filter=(TranslationFilter=(bX=true,bY=true,bZ=true),RotationFilter=(bX=true,bY=true,bZ=true),ScaleFilter=(bX=false,bY=false,bZ=false))
    Parents=((Item=(Type=Bone,Name="thigh_l"),Weight=0.2),(Item=(Type=Bone,Name="thigh_r"),Weight=0.2),(Item=(Type=Bone,Name="pelvis"),Weight=0.6))
    Weight=1
   )
   RigUnit_BeginExecution = &RigUnit_BeginExecution()
   n1 = &RigUnit_GetRelativeTransformForItem(Child=(Type=Bone,Name="pelvis"), Parent=(Type=Bone,Name="root"), bChildInitial=False, bParentInitial=False)
   RigVMVariableNode(
    Value=&RigVMFunction_MathDoubleMul(
     B=&RigVMFunction_MathDoubleRemap(SourceMaximum=67, SourceMinimum=90, TargetMaximum=1, TargetMinimum=0, Value=@n1.RelativeTransform.Translation.Z).Result
     A=&RigVMFunction_MathDoubleRemap(SourceMaximum=69, SourceMinimum=50, TargetMaximum=1, TargetMinimum=0, Value=@n1.RelativeTransform.Translation.Z).Result
    ).Result
    ExecuteContext=@RigUnit_BeginExecution.ExecutePin
   ) {variable=Crouch}
   n2 = &RigUnit_GetTransform(Item=(Type=Null,Name="calf_l_null"), bInitial=False)
   n3 = &RigUnit_GetTransform(Item=(Type=Null,Name="calf_r_null"), bInitial=False)
   > RigVMVariableNode(Value=&RigVMFunction_MathDoubleRemap(
     SourceMaximum=30
     SourceMinimum=40
     TargetMaximum=1
     TargetMinimum=0
     Value=&RigVMFunction_MathVectorDistance(B=@n2.Transform.Translation, A=@n3.Transform.Translation).Result
    ).Result) {variable=Dist-ce Between knee}
   > FN_DeformRig_FNC = RigVMFunctionReferenceNode(
    DeformRig_ Curve="(deformRig_l_upperarm_up_alpha_72_#1=False,deformRig_r_upperarm_up_alpha_73_#2=False,deformRig_l_upperarm_fr_alpha_74_#3=False,deformRig_r_upperarm_fr_alpha_75_CB51B8 ...(truncated)"
    DeformRig_Bool="(upperarm_1_#4=True,elbow_3_#5=True,wrist_7_#6=True,groin_8_#7=True,glute_17_#8=True,knee_9_40C408324CA261 ...(truncated)"
    DeformRig_Debug=(upperarm_debug_18_#4=False,wrist_debug_20_#6=False,lat_debug_26_#9=False,groinGlute_debug_27_#7=False)
    DeformRig_Value="(upperarm_up_18_#10=(Rotation=(X=0,Y=0,Z=0,W=1),Translation=(X=0,Y=0,Z=8),Scale3D=(X=0.7,Y=1,Z=1)),upperarm_fwd_fr_19_#11=(Rotation=(X=0,Y=0,Z=0,W=1),Translation=(X=0,Y=10,Z=0),Scale3D=(X=0.5,Y=1,Z= ...(truncated)"
   )
   > RigVMFunction_Sequence_2_1 = RigVMAggregateNode()
   &RigUnit_ModifyTransforms(
    ItemToModify="((Item=(Type=Bone,Name=\"dyn_master_sleeve\"),Transform=(Rotation=(X=0,Y=0,Z=0,W=1),Translation=(X=4,Y=0,Z=0),Scale3D=(X=1,Y=1,Z=1))),(Item=(Type=Bone,Name=\"base_sleeve_l\"),Transform=(Rotation=(X=0,Y=0,Z=0,W=1),Translation=(X=0,Y=-2,Z=0),Scale3D=(X=1,Y=1,Z=1 ...(truncated)"
    WeightMinimum=0
    Weight=&RigVMFunction_MathDoubleRemap(
     SourceMaximum=50
     SourceMinimum=70
     TargetMaximum=1
     TargetMinimum=0
     Value=&RigUnit_GetRelativeTransformForItem(Child=(Type=Bone,Name="pelvis"), Parent=(Type=Bone,Name="root"), bChildInitial=False, bParentInitial=False).RelativeTransform.Translation.Z
    ).Result
    ExecutePin=@RigVMFunction_Sequence_2_1.A
   )
   n4 = &RigUnit_SphericalPoseReader(
    ActiveRegionScaleFactors=(PositiveWidth=0.6,NegativeWidth=1.4,PositiveHeight=1,NegativeHeight=1)
    ActiveRegionSize=0.04
    Debug=(bDrawDebug=false,bDraw2D=false,bDrawLocalAxes=false,DebugScale=30,DebugSegments=30,DebugThickness=0.2)
    DriverAxis=(X=1,Y=0,Z=0)
    DriverItem=(Type=Bone,Name="thigh_r")
    FalloffRegionScaleFactors=(PositiveWidth=1.2,NegativeWidth=1.4,PositiveHeight=1,NegativeHeight=1)
    FalloffSize=0.1
    FlipWidthScaling=true
    OptionalParentItem=(Type=Bone,Name="pelvis")
    RotationOffset=(X=-120,Y=180,Z=-20)
    ExecutePin=@RigVMFunction_Sequence_2_1.C
   )
   n5 = &RigVMFunction_AlphaInterp(InterpSpeedDecreasing=14, InterpSpeedIncreasing=14, Scale=1, bClampResult=False, bInterpResult=False, Value=@n4.OutputParam)
   > &RigUnit_ModifyTransforms(
    ItemToModify=((Item=(Type=Bone,Name="dyn_master_sleeve"),Transform=(Rotation=(X=0.1164,Y=-0.0347,Z=0.0041,W=0.9926),Translation=(X=4.8011,Y=-2.89,Z=-2.1433),Scale3D=(X=1,Y=1,Z=1))))
    WeightMinimum=0
    Weight=@n5.Result
   )
   &RigUnit_SphericalPoseReader(
    ActiveRegionScaleFactors=(PositiveWidth=1,NegativeWidth=0.1,PositiveHeight=1,NegativeHeight=1)
    ActiveRegionSize=0.3
    Debug=(bDrawDebug=false,bDraw2D=false,bDrawLocalAxes=false,DebugScale=30,DebugSegments=30,DebugThickness=0.2)
    DriverAxis=(X=-1,Y=0,Z=0)
    DriverItem=(Type=Bone,Name="thigh_l")
    FalloffRegionScaleFactors=(PositiveWidth=1,NegativeWidth=1,PositiveHeight=1,NegativeHeight=1)
    FalloffSize=0.5
    OptionalParentItem=(Type=Bone,Name="pelvis")
    RotationOffset=(X=100,Y=0,Z=-7)
    ExecutePin=@RigVMFunction_Sequence_2_1.D
   )
   n6 = &RigUnit_GetRelativeTransformForItem(Child=(Type=Bone,Name="pelvis"), Parent=(Type=Bone,Name="root"), bChildInitial=False, bParentInitial=False)
   > &RigUnit_ModifyTransforms(
    ItemToModify="((Item=(Type=Bone,Name=\"dyn_master_sleeve\"),Transform=(Rotation=(X=0,Y=-0.1736,Z=0,W=0.9848),Translation=(X=5,Y=0,Z=0),Scale3D=(X=1,Y=1,Z=1))),(Item=(Type=Bone,Name=\"base_sleeve_l\"),Transform=(Rotation=(X=0.1736,Y=0,Z=0,W=0.9848),Translation=(X=0,Y=0,Z=0), ...(truncated)"
    WeightMinimum=0
    Weight=&RigVMFunction_MathDoubleMul(
     B=&RigVMFunction_MathDoubleRemap(SourceMaximum=67, SourceMinimum=90, TargetMaximum=1, TargetMinimum=0, Value=@n6.RelativeTransform.Translation.Z).Result
     A=&RigVMFunction_MathDoubleRemap(SourceMaximum=69, SourceMinimum=50, TargetMaximum=1, TargetMinimum=0, Value=@n6.RelativeTransform.Translation.Z).Result
    ).Result
   )
   &RigUnit_SphericalPoseReader(
    ActiveRegionScaleFactors=(PositiveWidth=1,NegativeWidth=0.01,PositiveHeight=1,NegativeHeight=1)
    ActiveRegionSize=0.3
    Debug=(bDrawDebug=false,bDraw2D=false,bDrawLocalAxes=false,DebugScale=30,DebugSegments=30,DebugThickness=0.2)
    DriverAxis=(X=1,Y=0,Z=0)
    DriverItem=(Type=Bone,Name="thigh_r")
    FalloffRegionScaleFactors=(PositiveWidth=1,NegativeWidth=1,PositiveHeight=1,NegativeHeight=1)
    FalloffSize=0.5
    OptionalParentItem=(Type=Bone,Name="pelvis")
    RotationOffset=(X=-92,Y=180,Z=-2)
    ExecutePin=@RigVMFunction_Sequence_2_1.E
   )
   n7 = &RigUnit_GetRelativeTransformForItem(Child=(Type=Bone,Name="pelvis"), Parent=(Type=Bone,Name="root"), bChildInitial=False, bParentInitial=False)
   > &RigUnit_ModifyTransforms(
    ItemToModify="((Item=(Type=Bone,Name=\"dyn_master_sleeve\"),Transform=(Rotation=(X=0,Y=0.1736,Z=0,W=0.9848),Translation=(X=3,Y=0,Z=0),Scale3D=(X=1,Y=1,Z=1))),(Item=(Type=Bone,Name=\"base_sleeve_r\"),Transform=(Rotation=(X=0,Y=0,Z=0,W=1),Translation=(X=4,Y=0,Z=0),Scale3D=(X= ...(truncated)"
    WeightMinimum=0
    Weight=&RigVMFunction_MathDoubleMul(
     B=&RigVMFunction_MathDoubleRemap(SourceMaximum=67, SourceMinimum=90, TargetMaximum=1, TargetMinimum=0, Value=@n7.RelativeTransform.Translation.Z).Result
     A=&RigVMFunction_MathDoubleRemap(SourceMaximum=69, SourceMinimum=50, TargetMaximum=1, TargetMinimum=0, Value=@n7.RelativeTransform.Translation.Z).Result
    ).Result
   )
   n8 = &RigVMFunction_Sequence(ExecuteContext=@RigVMFunction_Sequence_2_1.F)
   &RigUnit_SphericalPoseReader(
    ActiveRegionScaleFactors=(PositiveWidth=1,NegativeWidth=0.1,PositiveHeight=1,NegativeHeight=1)
    ActiveRegionSize=0.3
    Debug=(bDrawDebug=false,bDraw2D=false,bDrawLocalAxes=false,DebugScale=30,DebugSegments=30,DebugThickness=0.2)
    DriverAxis=(X=1,Y=0,Z=0)
    DriverItem=(Type=Bone,Name="thigh_r")
    FalloffRegionScaleFactors=(PositiveWidth=1,NegativeWidth=1,PositiveHeight=1,NegativeHeight=1)
    FalloffSize=0.5
    OptionalParentItem=(Type=Bone,Name="pelvis")
    RotationOffset=(X=-90,Y=180,Z=-5)
    ExecutePin=@n8.A
   )
   > &RigUnit_ModifyTransforms(
    ItemToModify=((Item=(Type=Bone,Name="base_sleeve_r"),Transform=(Rotation=(X=0,Y=-0.6686,Z=0,W=0.7437),Translation=(X=-4.5293,Y=0.5495,Z=1.4404),Scale3D=(X=1,Y=1,Z=1))))
    WeightMinimum=0
    Weight=RigVMDispatchNode(BoneName=root, Default=0, Name=deformRig_r_thigh_up_alpha) {template="DISPATCH_RigDispatch_GetAnimAttribute(in Name,in BoneName,in Default,out Value,out Found)"}.Value
   )
   &RigUnit_SphericalPoseReader(
    ActiveRegionScaleFactors=(PositiveWidth=1,NegativeWidth=0.1,PositiveHeight=1,NegativeHeight=1)
    ActiveRegionSize=0.3
    Debug=(bDrawDebug=false,bDraw2D=false,bDrawLocalAxes=false,DebugScale=30,DebugSegments=30,DebugThickness=0.2)
    DriverAxis=(X=-1,Y=0,Z=0)
    DriverItem=(Type=Bone,Name="thigh_l")
    FalloffRegionScaleFactors=(PositiveWidth=1,NegativeWidth=1,PositiveHeight=1,NegativeHeight=1)
    FalloffSize=0.5
    OptionalParentItem=(Type=Bone,Name="pelvis")
    RotationOffset=(X=115,Y=0,Z=-30)
    ExecutePin=@n8.B
   )
   DISPATCH_RigDispatch_GetAnimAttribute = RigVMDispatchNode(BoneName=root, Default=0, Name=deformRig_l_thigh_up_alpha) {template="DISPATCH_RigDispatch_GetAnimAttribute(in Name,in BoneName,in Default,out Value,out Found)"}
   > &RigUnit_ModifyTransforms(
    ItemToModify=((Item=(Type=Bone,Name="base_sleeve_l"),Transform=(Rotation=(X=0,Y=-0.6315,Z=0,W=0.7754),Translation=(X=-6,Y=0,Z=0),Scale3D=(X=1,Y=1,Z=1))))
    WeightMinimum=0
    Weight=@DISPATCH_RigDispatch_GetAnimAttribute.Value
   )
   &RigUnit_ParentConstraint(
    AdvancedSettings=(RotationOrderForFilter=XZY)
    Child=(Type=Bone,Name="Base_thingCorr_r")
    Filter=(TranslationFilter=(bX=false,bY=false,bZ=false),RotationFilter=(bX=true,bY=true,bZ=true),ScaleFilter=(bX=false,bY=false,bZ=false))
    Parents=((Item=(Type=Bone,Name="thigh_r"),Weight=0.8),(Item=(Type=Bone,Name="pelvis"),Weight=0.2))
    Weight=1
    bMaintainOffset=true
    ExecutePin=@RigVMFunction_Sequence_2_1.G
   )
   > &RigUnit_ParentConstraint(
    AdvancedSettings=(RotationOrderForFilter=XZY)
    Child=(Type=Bone,Name="Base_thingCorr_l")
    Filter=(TranslationFilter=(bX=false,bY=false,bZ=false),RotationFilter=(bX=true,bY=true,bZ=true),ScaleFilter=(bX=false,bY=false,bZ=false))
    Parents=((Item=(Type=Bone,Name="thigh_l"),Weight=0.8),(Item=(Type=Bone,Name="pelvis"),Weight=0.2))
    Weight=1
    bMaintainOffset=true
   )
   > &RigUnit_ParentConstraint(
    AdvancedSettings=(RotationOrderForFilter=XZY)
    Child=(Type=Bone,Name="Base_back")
    Filter=(TranslationFilter=(bX=true,bY=true,bZ=true),RotationFilter=(bX=true,bY=true,bZ=true),ScaleFilter=(bX=false,bY=false,bZ=false))
    Parents=((Item=(Type=Bone,Name="thigh_l"),Weight=0.4),(Item=(Type=Bone,Name="thigh_r"),Weight=0.4),(Item=(Type=Bone,Name="pelvis"),Weight=0.2))
    Weight=1
   )
   > n9 = &RigUnit_ModifyTransforms(
    ItemToModify="((Item=(Type=Bone,Name=\"Base_thingCorr_l\"),Transform=(Rotation=(X=0,Y=0,Z=0,W=1),Translation=(X=0,Y=0,Z=-1),Scale3D=(X=1,Y=1,Z=1))),(Item=(Type=Bone,Name=\"pelvis_scale_corr\"),Transform=(Rotation=(X=0,Y=0,Z=0,W=1),Translation=(X=0,Y=0,Z=-1),Scale3D=(X=1,Y=1 ...(truncated)"
    Weight=1
   )
   n10 = &RigVMFunction_ControlFlowBranch(ExecuteContext=@n9.ExecutePin, Condition=RigVMVariableNode() {variable=isEmote}.Value)
   RigVMFunction_Sequence_1 = RigVMAggregateNode(ExecuteContext=@n10.False)
   n11 = &RigUnit_SphericalPoseReader(
    ActiveRegionScaleFactors=(PositiveWidth=1,NegativeWidth=1,PositiveHeight=1,NegativeHeight=1)
    ActiveRegionSize=0.05
    Debug=(bDrawDebug=false,DebugScale=25,DebugSegments=20,DebugThickness=0.25)
    DriverAxis=(X=-1,Y=0,Z=0)
    DriverItem=(Type=Bone,Name="thigh_l")
    FalloffRegionScaleFactors=(PositiveWidth=1,NegativeWidth=1,PositiveHeight=1,NegativeHeight=1)
    FalloffSize=0.15
    FlipWidthScaling=false
    OptionalParentItem=(Type=Bone)
    RotationOffset=(X=110,Y=0,Z=-20)
    ExecutePin=@RigVMFunction_Sequence_1.A
   )
   > &RigUnit_ModifyTransforms(
    ItemToModify="((Item=(Type=Bone,Name=\"Base_back\"),Transform=(Rotation=(X=-0.0435,Y=0.0034,Z=0.0777,W=0.996),Translation=(X=0,Y=-3.9504,Z=5.3456),Scale3D=(X=1,Y=1,Z=1))),(Item=(Type=Bone,Name=\"Base_thingCorr_r\"),Transform=(Rotation=(X=0,Y=0.0436,Z=0,W=0.999),Translation= ...(truncated)"
    Weight=@n11.OutputParam
   )
   > n12 = &RigUnit_SphericalPoseReader(
    ActiveRegionScaleFactors=(PositiveWidth=3,NegativeWidth=1,PositiveHeight=3,NegativeHeight=3)
    ActiveRegionSize=0.05
    Debug=(bDrawDebug=false,DebugScale=25,DebugSegments=20,DebugThickness=0.25)
    DriverAxis=(X=-1,Y=0,Z=0)
    DriverItem=(Type=Bone,Name="thigh_l")
    FalloffRegionScaleFactors=(PositiveWidth=1,NegativeWidth=1,PositiveHeight=1,NegativeHeight=1)
    FalloffSize=0.4
    FlipHeightScaling=false
    OptionalParentItem=(Type=Bone)
    RotationOffset=(X=110,Y=-30,Z=0)
   )
   > &RigUnit_ModifyTransforms(
    ItemToModify="((Item=(Type=Bone,Name=\"Base_back\"),Transform=(Rotation=(X=-0.0584,Y=0.0926,Z=0.1684,W=0.9796),Translation=(X=6.2767,Y=-7.3105,Z=1.3278),Scale3D=(X=1,Y=1,Z=1))),(Item=(Type=Bone,Name=\"Base_thingCorr_l\"),Transform=(Rotation=(X=0,Y=0,Z=0,W=1),Translation=(X= ...(truncated)"
    Weight=@n12.OutputParam
   )
   > n13 = &RigUnit_SphericalPoseReader(
    ActiveRegionScaleFactors=(PositiveWidth=1,NegativeWidth=1,PositiveHeight=2.5,NegativeHeight=2.5)
    ActiveRegionSize=0.05
    Debug=(bDrawDebug=false,DebugScale=25,DebugSegments=20,DebugThickness=0.25)
    DriverAxis=(X=-1,Y=0,Z=0)
    DriverItem=(Type=Bone,Name="thigh_l")
    FalloffRegionScaleFactors=(PositiveWidth=1,NegativeWidth=1,PositiveHeight=1,NegativeHeight=1)
    FalloffSize=0.17
    FlipHeightScaling=false
    OptionalParentItem=(Type=Bone)
    RotationOffset=(X=-90,Y=0,Z=-69)
   )
   > &RigUnit_ModifyTransforms(
    ItemToModify="((Item=(Type=Bone,Name=\"Base_back\"),Transform=(Rotation=(X=-0.3338,Y=0,Z=0,W=0.9426),Translation=(X=-1.4812,Y=-1,Z=0),Scale3D=(X=1,Y=1,Z=1))),(Item=(Type=Bone,Name=\"thigh_corr_scale_l\"),Transform=(Rotation=(X=0,Y=0,Z=0,W=1),Translation=(X=0,Y=7,Z=-5),Scale ...(truncated)"
    Weight=@n13.OutputParam
   )
   n14 = &RigUnit_SphericalPoseReader(
    ActiveRegionScaleFactors=(PositiveWidth=1,NegativeWidth=1,PositiveHeight=1,NegativeHeight=1)
    ActiveRegionSize=0.05
    Debug=(bDrawDebug=false,DebugScale=25,DebugSegments=20,DebugThickness=0.25)
    DriverAxis=(X=1,Y=0,Z=0)
    DriverItem=(Type=Bone,Name="thigh_r")
    FalloffRegionScaleFactors=(PositiveWidth=1,NegativeWidth=1,PositiveHeight=1,NegativeHeight=1)
    FalloffSize=0.15
    FlipWidthScaling=false
    OptionalParentItem=(Type=Bone)
    RotationOffset=(X=-100,Y=100,Z=-16)
    ExecutePin=@RigVMFunction_Sequence_1.B
   )
   > &RigUnit_ModifyTransforms(
    ItemToModify="((Item=(Type=Bone,Name=\"Base_back\"),Transform=(Rotation=(X=-0.0697,Y=0.0435,Z=0.003,W=0.9966),Translation=(X=-2,Y=0,Z=0),Scale3D=(X=1,Y=1,Z=1))),(Item=(Type=Bone,Name=\"Base_thingCorr_l\"),Transform=(Rotation=(X=0,Y=0.0436,Z=0,W=0.999),Translation=(X=0,Y=0,Z ...(truncated)"
    Weight=@n14.OutputParam
   )
   > n15 = &RigUnit_SphericalPoseReader(
    ActiveRegionScaleFactors=(PositiveWidth=1,NegativeWidth=3,PositiveHeight=3,NegativeHeight=3)
    ActiveRegionSize=0.05
    Debug=(bDrawDebug=false,bDraw2D=false,DebugScale=25,DebugSegments=20,DebugThickness=0.25)
    DriverAxis=(X=1,Y=0,Z=0)
    DriverItem=(Type=Bone,Name="thigh_r")
    FalloffRegionScaleFactors=(PositiveWidth=1,NegativeWidth=1,PositiveHeight=1,NegativeHeight=1)
    FalloffSize=0.4
    FlipHeightScaling=false
    OptionalParentItem=(Type=Bone)
    RotationOffset=(X=-80,Y=0,Z=-10)
   )
   > &RigUnit_ModifyTransforms(
    ItemToModify="((Item=(Type=Bone,Name=\"Base_back\"),Transform=(Rotation=(X=0.0176,Y=0,Z=0,W=0.9998),Translation=(X=0,Y=0,Z=0),Scale3D=(X=1,Y=1,Z=1))),(Item=(Type=Bone,Name=\"Base_thingCorr_r\"),Transform=(Rotation=(X=0,Y=0,Z=0,W=1),Translation=(X=-6,Y=0,Z=-3.5),Scale3D=(X=1 ...(truncated)"
    Weight=@n15.OutputParam
   )
   > n16 = &RigUnit_SphericalPoseReader(
    ActiveRegionScaleFactors=(PositiveWidth=1,NegativeWidth=1,PositiveHeight=2.5,NegativeHeight=2.5)
    ActiveRegionSize=0.05
    Debug=(bDrawDebug=false,DebugScale=25,DebugSegments=20,DebugThickness=0.25)
    DriverAxis=(X=1,Y=0,Z=0)
    DriverItem=(Type=Bone,Name="thigh_r")
    FalloffRegionScaleFactors=(PositiveWidth=1,NegativeWidth=1,PositiveHeight=1,NegativeHeight=1)
    FalloffSize=0.15
    FlipHeightScaling=false
    OptionalParentItem=(Type=Bone)
    RotationOffset=(X=90,Y=0,Z=-69)
   )
   > &RigUnit_ModifyTransforms(
    ItemToModify="((Item=(Type=Bone,Name=\"Base_back\"),Transform=(Rotation=(X=0,Y=0,Z=0,W=1),Translation=(X=0,Y=0,Z=0),Scale3D=(X=1,Y=1,Z=1))),(Item=(Type=Bone,Name=\"thigh_corr_scale_r\"),Transform=(Rotation=(X=0,Y=0,Z=0,W=1),Translation=(X=0,Y=3.3339,Z=70.111),Scale3D=(X=1,Y ...(truncated)"
    Weight=@n16.OutputParam
   )
   > n17 = &RigUnit_SphericalPoseReader(
    ActiveRegionScaleFactors=(PositiveWidth=3,NegativeWidth=1,PositiveHeight=3,NegativeHeight=3)
    ActiveRegionSize=0.05
    Debug=(bDrawDebug=false,DebugScale=25,DebugSegments=20,DebugThickness=0.25)
    DriverAxis=(X=-1,Y=0,Z=0)
    DriverItem=(Type=Bone,Name="thigh_l")
    FalloffRegionScaleFactors=(PositiveWidth=1,NegativeWidth=1,PositiveHeight=1,NegativeHeight=1)
    FalloffSize=0.4
    FlipHeightScaling=false
    OptionalParentItem=(Type=Bone)
    RotationOffset=(X=110,Y=-30,Z=0)
   )
   > n18 = &RigUnit_SphericalPoseReader(
    ActiveRegionScaleFactors=(PositiveWidth=1,NegativeWidth=3,PositiveHeight=3,NegativeHeight=3)
    ActiveRegionSize=0.05
    Debug=(bDrawDebug=false,bDraw2D=false,DebugScale=25,DebugSegments=20,DebugThickness=0.25)
    DriverAxis=(X=1,Y=0,Z=0)
    DriverItem=(Type=Bone,Name="thigh_r")
    FalloffRegionScaleFactors=(PositiveWidth=1,NegativeWidth=1,PositiveHeight=1,NegativeHeight=1)
    FalloffSize=0.4
    FlipHeightScaling=false
    OptionalParentItem=(Type=Bone)
    RotationOffset=(X=-80,Y=0,Z=-10)
   )
   Multiply = &RigVMFunction_MathDoubleMul(B=@n18.OutputParam, A=@n17.OutputParam)
   > n19 = &RigUnit_ModifyTransforms(
    ItemToModify=((Item=(Type=Bone,Name="skirt_base"),Transform=(Rotation=(X=0,Y=0,Z=-0.2031,W=0.9792),Translation=(X=0,Y=-3,Z=0),Scale3D=(X=1,Y=1,Z=1))))
    Weight=@Multiply.Result
   )
   n20 = &RigVMFunction_Sequence(ExecuteContext=@n19.ExecutePin)
   n21 = RigVMVariableNode() {variable=Left Right Movement}
   RigVMVariableNode(
    Value=&RigVMFunction_KalmanFloat(
     BufferSize=8
     Value=&RigVMFunction_MathDoubleRemap(SourceMaximum=0, SourceMinimum=-1, TargetMaximum=0, TargetMinimum=1, Value=@n21.Value).Result
    ) {template="KalmanFilter::Execute(in Value,in BufferSize,out Result)"}.Result
    ExecuteContext=@n20.B
   ) {variable=leftM}
   > RigVMVariableNode(Value=&RigVMFunction_KalmanFloat(
     BufferSize=8
     Value=&RigVMFunction_MathDoubleRemap(SourceMaximum=1, SourceMinimum=0, TargetMaximum=1, TargetMinimum=0, Value=@n21.Value).Result
    ) {template="KalmanFilter::Execute(in Value,in BufferSize,out Result)"}.Result) {variable=Right Movement}
   n22 = RigVMVariableNode() {variable=Forward Backward Movement}
   Remap = &RigVMFunction_MathDoubleRemap(SourceMaximum=1, SourceMinimum=0, TargetMaximum=1, TargetMinimum=0, Value=@n22.Value)
   KalmanFilter = &RigVMFunction_KalmanFloat(BufferSize=8, Value=@Remap.Result) {template="KalmanFilter::Execute(in Value,in BufferSize,out Result)"}
   RigVMVariableNode(Value=@KalmanFilter.Result, ExecuteContext=@n20.A) {variable=Forward Movement}
   > n23 = RigVMVariableNode(Value=&RigVMFunction_KalmanFloat(
     BufferSize=8
     Value=&RigVMFunction_MathDoubleRemap(SourceMaximum=0, SourceMinimum=-1, TargetMaximum=0, TargetMinimum=1, Value=@n22.Value).Result
    ) {template="KalmanFilter::Execute(in Value,in BufferSize,out Result)"}.Result) {variable=Backward Movement}
   n24 = &RigVMFunction_ControlFlowBranch(Condition=RigVMVariableNode() {variable=isSkyDive}.Value, ExecuteContext=@n23.ExecuteContext)
   Multiply_1 = RigVMAggregateNode(
    B=RigVMVariableNode() {variable=Crouch}.Value
    A=RigVMVariableNode() {variable=Dist-ce Between knee}.Value
    C=RigVMVariableNode() {variable=leftM}.Value
   )
   &RigUnit_ModifyTransforms(
    ItemToModify="((Item=(Type=Bone,Name=\"Base_back\"),Transform=(Rotation=(X=0,Y=0,Z=0,W=1),Translation=(X=8.5049,Y=-1.6394,Z=0),Scale3D=(X=1,Y=1,Z=1))),(Item=(Type=Bone,Name=\"Base_thingCorr_l\"),Transform=(Rotation=(X=0,Y=0,Z=0,W=1),Translation=(X=-6.0356,Y=-10.5961,Z=0.000 ...(truncated)"
    Weight=@Multiply_1.Result
    ExecutePin=@n24.False
   )
   n25 = RigVMVariableNode() {variable=leftM}
   n26 = RigVMVariableNode() {variable=Forward Movement}
   Multiply_2 = RigVMAggregateNode(
    C=&RigVMFunction_MathDoubleRemap(SourceMaximum=0.5, SourceMinimum=0, TargetMaximum=1, TargetMinimum=0, Value=@n25.Value).Result
    A=&RigVMFunction_MathDoubleRemap(SourceMaximum=0.5, SourceMinimum=0, TargetMaximum=1, TargetMinimum=0, Value=@n26.Value).Result
    D=&RigVMFunction_MathDoubleRemap(SourceMaximum=0.71, SourceMinimum=1, TargetMaximum=1, TargetMinimum=0, Value=@n25.Value).Result
    B=&RigVMFunction_MathDoubleRemap(SourceMaximum=0.71, SourceMinimum=1, TargetMaximum=1, TargetMinimum=0, Value=@n26.Value).Result
   )
   TimeOffset = /Script/RigVM.RigVMFunction_TimeOffsetFloat(BufferSize=20, SecondsAgo=0.5, TimeRange=1, Value=RigVMVariableNode() {variable=Crouch}.Value) {is=struct RigVMUnitNode, template="TimeOffset::Execute(in Value,in SecondsAgo,in BufferSize,in TimeRange,out Result)"}
   > &RigUnit_ModifyTransforms(
    ItemToModify="((Item=(Type=Bone,Name=\"Base_back\"),Transform=(Rotation=(X=0,Y=0,Z=0,W=1),Translation=(X=-8.905,Y=0,Z=0),Scale3D=(X=1,Y=1,Z=1))),(Item=(Type=Bone,Name=\"Base_thingCorr_r\"),Transform=(Rotation=(X=0,Y=0,Z=0,W=1),Translation=(X=0,Y=7.2386,Z=0),Scale3D=(X=1,Y=1 ...(truncated)"
    Weight=&RigVMFunction_MathDoubleMul(A=@Multiply_2.Result, B=@TimeOffset.Result).Result
   )
   n27 = RigVMVariableNode() {variable=Forward Movement}
   n28 = RigVMVariableNode() {variable=Right Movement}
   Multiply_2_1 = RigVMAggregateNode(
    A=&RigVMFunction_MathDoubleRemap(SourceMaximum=0.5, SourceMinimum=0, TargetMaximum=1, TargetMinimum=0, Value=@n27.Value).Result
    C=&RigVMFunction_MathDoubleRemap(SourceMaximum=0.5, SourceMinimum=0, TargetMaximum=1, TargetMinimum=0, Value=@n28.Value).Result
    B=&RigVMFunction_MathDoubleRemap(SourceMaximum=0.7, SourceMinimum=1, TargetMaximum=1, TargetMinimum=0, Value=@n27.Value).Result
    D=&RigVMFunction_MathDoubleRemap(SourceMaximum=0.7, SourceMinimum=1, TargetMaximum=1, TargetMinimum=0, Value=@n28.Value).Result
   )
   > &RigUnit_ModifyTransforms(
    ItemToModify="((Item=(Type=Bone,Name=\"Base_back\"),Transform=(Rotation=(X=0,Y=0,Z=0,W=1),Translation=(X=6.1084,Y=-4.0277,Z=0),Scale3D=(X=1,Y=1,Z=1))),(Item=(Type=Bone,Name=\"Base_thingCorr_r\"),Transform=(Rotation=(X=0,Y=0,Z=0,W=1),Translation=(X=0,Y=0,Z=0),Scale3D=(X=1,Y= ...(truncated)"
    Weight=&RigVMFunction_MathDoubleMul(B=@Multiply_2_1.Result, A=RigVMVariableNode() {variable=Crouch}.Value).Result
   )
   n29 = &RigUnit_GetRelativeTransformForItem(Child=(Type=Bone,Name="pelvis"), Parent=(Type=Bone,Name="root"), bChildInitial=False, bParentInitial=False)
   Multiply_1_1 = RigVMAggregateNode(
    C=&RigVMFunction_MathDoubleRemap(
     SourceMaximum=90
     SourceMinimum=70
     TargetMaximum=1
     TargetMinimum=0
     Value=&RigVMFunction_MathQuaternionToEuler(RotationOrder=ZYX, Value=@n29.RelativeTransform.Rotation).Result.Z
    ).Result
    A=RigVMVariableNode() {variable=Crouch}.Value
    B=RigVMVariableNode() {variable=Right Movement}.Value
   )
   > &RigUnit_ModifyTransforms(
    ItemToModify="((Item=(Type=Bone,Name=\"Base_back\"),Transform=(Rotation=(X=0,Y=-0.1321,Z=0,W=0.9912),Translation=(X=-0.3992,Y=2.9648,Z=-4.5201),Scale3D=(X=1,Y=1,Z=1))),(Item=(Type=Bone,Name=\"Base_thingCorr_r\"),Transform=(Rotation=(X=0,Y=0,Z=0,W=1),Translation=(X=-1.9575,Y ...(truncated)"
    Weight=@Multiply_1_1.Result
   )
   n30 = RigVMVariableNode() {variable=Backward Movement}
   n31 = RigVMVariableNode() {variable=Right Movement}
   Multiply_2_1_1 = RigVMAggregateNode(
    A=&RigVMFunction_MathDoubleRemap(SourceMaximum=0.5, SourceMinimum=0, TargetMaximum=1, TargetMinimum=0, Value=@n30.Value).Result
    C=&RigVMFunction_MathDoubleRemap(SourceMaximum=0.5, SourceMinimum=0, TargetMaximum=1, TargetMinimum=0, Value=@n31.Value).Result
    B=&RigVMFunction_MathDoubleRemap(SourceMaximum=0.7, SourceMinimum=1, TargetMaximum=1, TargetMinimum=0, Value=@n30.Value).Result
    D=&RigVMFunction_MathDoubleRemap(SourceMaximum=0.7, SourceMinimum=1, TargetMaximum=1, TargetMinimum=0, Value=@n31.Value).Result
   )
   > &RigUnit_ModifyTransforms(
    ItemToModify="((Item=(Type=Bone,Name=\"Base_back\"),Transform=(Rotation=(X=0,Y=0,Z=0,W=1),Translation=(X=-6.5348,Y=0,Z=0),Scale3D=(X=1,Y=1,Z=1))),(Item=(Type=Bone,Name=\"Base_thingCorr_r\"),Transform=(Rotation=(X=0,Y=0,Z=0,W=1),Translation=(X=0,Y=5.6357,Z=0),Scale3D=(X=1,Y= ...(truncated)"
    Weight=&RigVMFunction_MathDoubleMul(B=@Multiply_2_1_1.Result, A=RigVMVariableNode() {variable=Crouch}.Value).Result
   )
   > Get Delta Ref = RigVMFunctionReferenceNode(Item=(Type=Bone,Name="pelvis"))
   Multiply_1_1_2 = RigVMAggregateNode(
    C=&RigVMFunction_MathDoubleRemap(SourceMaximum=50, SourceMinimum=80, TargetMaximum=1, TargetMinimum=0, Value=@"Get Delta Ref".Result.Rotation.Yaw).Result
    A=RigVMVariableNode() {variable=Crouch}.Value
    B=RigVMVariableNode() {variable=leftM}.Value
   )
   > &RigUnit_ModifyTransforms(
    ItemToModify="((Item=(Type=Bone,Name=\"Base_back\"),Transform=(Rotation=(X=0.1255,Y=-0.133,Z=0.011,W=0.9831),Translation=(X=0,Y=-10.3179,Z=0),Scale3D=(X=1,Y=1,Z=1))),(Item=(Type=Bone,Name=\"Base_thingCorr_r\"),Transform=(Rotation=(X=0,Y=0,Z=0,W=1),Translation=(X=0,Y=0,Z=0), ...(truncated)"
    Weight=@Multiply_1_1_2.Result
   )
   > n32 = &RigUnit_SphericalPoseReader(
    ActiveRegionScaleFactors=(PositiveWidth=1,NegativeWidth=1,PositiveHeight=1.8,NegativeHeight=1)
    ActiveRegionSize=0.05
    Debug=(bDrawDebug=false,DebugScale=25,DebugSegments=20,DebugThickness=0.25)
    DriverAxis=(X=-1,Y=0,Z=0)
    DriverItem=(Type=Bone,Name="thigh_l")
    FalloffRegionScaleFactors=(PositiveWidth=1,NegativeWidth=1,PositiveHeight=1,NegativeHeight=1)
    FalloffSize=0.17
    OptionalParentItem=(Type=Bone)
    RotationOffset=(X=78,Y=0,Z=50)
   )
   > Get Delta Ref_1 = RigVMFunctionReferenceNode(Item=(Type=Bone,Name="pelvis"))
   Multiply_1_1_2_1 = RigVMAggregateNode(
    D=&RigVMFunction_MathDoubleRemap(SourceMaximum=50, SourceMinimum=80, TargetMaximum=1, TargetMinimum=0, Value=@"Get Delta Ref_1".Result.Rotation.Yaw).Result
    C=@n32.OutputParam
    A=RigVMVariableNode() {variable=Crouch}.Value
    B=RigVMVariableNode() {variable=leftM}.Value
   )
   > &RigUnit_ModifyTransforms(
    ItemToModify="((Item=(Type=Bone,Name=\"Base_back\"),Transform=(Rotation=(X=0,Y=0,Z=0,W=1),Translation=(X=-9.4539,Y=-3.2127,Z=-0.0002),Scale3D=(X=1,Y=1,Z=1))),(Item=(Type=Bone,Name=\"Base_thingCorr_r\"),Transform=(Rotation=(X=0,Y=0,Z=0,W=1),Translation=(X=0,Y=5.3943,Z=0),Sca ...(truncated)"
    Weight=@Multiply_1_1_2_1.Result
   )
   &RigUnit_ModifyTransforms(
    ItemToModify="((Item=(Type=Bone,Name=\"skirt_base\"),Transform=(Rotation=(X=0,Y=0,Z=-0.1132,W=0.9936),Translation=(X=-0.5892,Y=-2.5521,Z=0),Scale3D=(X=1,Y=1,Z=1))),(Item=(Type=Bone,Name=\"dyn_skirt_bk_r_mstr\"),Transform=(Rotation=(X=0,Y=-0.0436,Z=0,W=0.999),Translation=(X= ...(truncated)"
    Weight=1
    WeightMinimum=0
    ExecutePin=@n24.True
   )
   > n33 = &RigUnit_SphericalPoseReader(
    ActiveRegionScaleFactors=(PositiveWidth=1,NegativeWidth=1,PositiveHeight=2.5,NegativeHeight=2.5)
    ActiveRegionSize=0.05
    Debug=(bDrawDebug=false,DebugScale=25,DebugSegments=20,DebugThickness=0.25)
    DriverAxis=(X=-1,Y=0,Z=0)
    DriverItem=(Type=Bone,Name="thigh_l")
    FalloffRegionScaleFactors=(PositiveWidth=1,NegativeWidth=1,PositiveHeight=1,NegativeHeight=1)
    FalloffSize=0.15
    FlipHeightScaling=false
    FlipWidthScaling=false
    OptionalParentItem=(Type=Bone)
    RotationOffset=(X=-90,Y=0,Z=-69)
   )
   > n34 = &RigUnit_SphericalPoseReader(
    ActiveRegionScaleFactors=(PositiveWidth=1,NegativeWidth=1,PositiveHeight=2.5,NegativeHeight=2.5)
    ActiveRegionSize=0.05
    Debug=(bDrawDebug=false,DebugScale=25,DebugSegments=20,DebugThickness=0.25)
    DriverAxis=(X=1,Y=0,Z=0)
    DriverItem=(Type=Bone,Name="thigh_r")
    FalloffRegionScaleFactors=(PositiveWidth=1,NegativeWidth=1,PositiveHeight=1,NegativeHeight=1)
    FalloffSize=0.15
    FlipHeightScaling=false
    OptionalParentItem=(Type=Bone)
    RotationOffset=(X=90,Y=0,Z=-69)
   )
   > &RigUnit_ModifyTransforms(
    ItemToModify="((Item=(Type=Bone,Name=\"Base_back\"),Transform=(Rotation=(X=0.042,Y=0,Z=0,W=0.9991),Translation=(X=0,Y=11.989,Z=0),Scale3D=(X=1,Y=1,Z=1))),(Item=(Type=Bone,Name=\"thigh_corr_scale_r\"),Transform=(Rotation=(X=0,Y=0,Z=0,W=1),Translation=(X=0,Y=0,Z=0),Scale3D=(X ...(truncated)"
    Weight=&RigVMFunction_MathDoubleMul(A=@n34.OutputParam, B=@n33.OutputParam).Result
   )
   RigVMFunction_ControlFlowBranch = &RigVMFunction_ControlFlowBranch(ExecuteContext=@RigVMFunction_Sequence_1.C, Condition=RigVMVariableNode() {variable=isAltskin}.Value)
   n35 = &RigUnit_SphericalPoseReader(
    ActiveRegionScaleFactors=(PositiveWidth=3,NegativeWidth=1,PositiveHeight=3,NegativeHeight=3)
    ActiveRegionSize=0.05
    Debug=(bDrawDebug=false,DebugScale=25,DebugSegments=20,DebugThickness=0.25)
    DriverAxis=(X=-1,Y=0,Z=0)
    DriverItem=(Type=Bone,Name="thigh_l")
    FalloffRegionScaleFactors=(PositiveWidth=1,NegativeWidth=1,PositiveHeight=1,NegativeHeight=1)
    FalloffSize=0.4
    FlipHeightScaling=false
    OptionalParentItem=(Type=Bone)
    RotationOffset=(X=110,Y=-30,Z=0)
    ExecutePin=@RigVMFunction_ControlFlowBranch.True
   )
   > &RigUnit_ModifyTransforms(
    ItemToModify="((Item=(Type=Bone,Name=\"dyn_back\"),Transform=(Rotation=(X=0,Y=0.1171,Z=0,W=0.9931),Translation=(X=-3.751,Y=0.8942,Z=-3.2604),Scale3D=(X=1,Y=1,Z=1))),(Item=(Type=Bone,Name=\"Base_thingCorr_r\"),Transform=(Rotation=(X=0,Y=0,Z=0,W=1),Translation=(X=0,Y=0,Z=0),S ...(truncated)"
    WeightMinimum=0
    Weight=@n35.OutputParam
   )
   > n36 = &RigUnit_SphericalPoseReader(
    ActiveRegionScaleFactors=(PositiveWidth=1,NegativeWidth=3,PositiveHeight=3,NegativeHeight=3)
    ActiveRegionSize=0.05
    Debug=(bDrawDebug=true,bDraw2D=false,DebugScale=25,DebugSegments=20,DebugThickness=0.25)
    DriverAxis=(X=1,Y=0,Z=0)
    DriverItem=(Type=Bone,Name="thigh_r")
    FalloffRegionScaleFactors=(PositiveWidth=1,NegativeWidth=1,PositiveHeight=1,NegativeHeight=1)
    FalloffSize=0.4
    FlipHeightScaling=false
    OptionalParentItem=(Type=Bone)
    RotationOffset=(X=-80,Y=0,Z=-10)
   )
   > &RigUnit_ModifyTransforms(
    ItemToModify=((Item=(Type=Bone,Name="Base_back"),Transform=(Rotation=(X=0,Y=0,Z=0,W=1),Translation=(X=-2,Y=0,Z=0),Scale3D=(X=1,Y=1,Z=1))))
    WeightMinimum=0
    Weight=@n36.OutputParam
   )
   > &RigUnit_ModifyTransforms(
    ItemToModify="((Item=(Type=Bone,Name=\"pelvis_scale_corr\"),Transform=(Rotation=(X=0,Y=0,Z=0,W=1),Translation=(X=0,Y=0,Z=-1.1521),Scale3D=(X=0.8,Y=0.8,Z=0.8))),(Item=(Type=Bone,Name=\"Base_back\"),Transform=(Rotation=(X=-0.0203,Y=0,Z=0,W=0.9998),Translation=(X=0,Y=0.0812,Z= ...(truncated)"
    Weight=1
   )
   > n37 = &RigUnit_ModifyTransforms(
    ItemToModify=((Item=(Type=Bone,Name="Base_back"),Transform=(Rotation=(X=0,Y=0,Z=0,W=1),Translation=(X=0,Y=13.6867,Z=3.391),Scale3D=(X=1,Y=1,Z=1))))
    WeightMinimum=0
    Weight=&RigVMFunction_MathDoubleRemap(
     SourceMaximum=50
     SourceMinimum=70
     TargetMaximum=1
     TargetMinimum=0
     Value=&RigUnit_GetRelativeTransformForItem(Child=(Type=Bone,Name="pelvis"), Parent=(Type=Bone,Name="root"), bChildInitial=False, bParentInitial=False).RelativeTransform.Translation.Z
    ).Result
   )
   n38 = &RigVMFunction_ControlFlowBranch(ExecuteContext=@n37.ExecutePin, Condition=RigVMVariableNode() {variable=isSkyDive}.Value)
   n39 = &RigUnit_GetRelativeTransformForItem(Child=(Type=Bone,Name="spine_04"), Parent=(Type=Bone,Name="spine_03"), bChildInitial=False, bParentInitial=False)
   n40 = &RigUnit_ModifyTransforms(
    ItemToModify=((Item=(Type=Bone,Name="Base_back"),Transform=(Rotation=(X=0,Y=0,Z=0,W=1),Translation=(X=0,Y=-7,Z=0),Scale3D=(X=1,Y=1,Z=1))))
    WeightMinimum=0
    Weight=&RigVMFunction_MathDoubleRemap(
     SourceMaximum=0
     SourceMinimum=10
     TargetMaximum=1
     TargetMinimum=0
     Value=&RigVMFunction_MathQuaternionToEuler(RotationOrder=XYZ, Value=@n39.RelativeTransform.Rotation).Result.Z
    ).Result
    ExecutePin=@n38.True
   )
   n41 = &RigVMFunction_ControlFlowBranch(ExecuteContext=@n40.ExecutePin, Condition=RigVMVariableNode() {variable=isFloating}.Value)
   &RigUnit_ModifyTransforms(
    ItemToModify="((Item=(Type=Bone,Name=\"pelvis_scale_corr\"),Transform=(Rotation=(X=0,Y=0,Z=0,W=1),Translation=(X=0,Y=0,Z=0),Scale3D=(X=1,Y=1,Z=1))),(Item=(Type=Bone,Name=\"thigh_corr_l\"),Transform=(Rotation=(X=0,Y=0,Z=0,W=1),Translation=(X=0,Y=0,Z=0),Scale3D=(X=1,Y=1,Z=1)) ...(truncated)"
    Weight=1
    WeightMinimum=0
    ExecutePin=@n41.True
   )
   > n42 = &RigUnit_SphericalPoseReader(
    ActiveRegionScaleFactors=(PositiveWidth=1,NegativeWidth=1,PositiveHeight=2.5,NegativeHeight=2.5)
    ActiveRegionSize=0.05
    Debug=(bDrawDebug=true,DebugScale=25,DebugSegments=20,DebugThickness=0.25)
    DriverAxis=(X=1,Y=0,Z=0)
    DriverItem=(Type=Bone,Name="thigh_r")
    FalloffRegionScaleFactors=(PositiveWidth=1,NegativeWidth=1,PositiveHeight=1,NegativeHeight=1)
    FalloffSize=0.15
    FlipHeightScaling=false
    OptionalParentItem=(Type=Bone)
    RotationOffset=(X=90,Y=0,Z=-69)
   )
   > &RigUnit_ModifyTransforms(
    ItemToModify="((Item=(Type=Bone,Name=\"Base_back\"),Transform=(Rotation=(X=0,Y=0,Z=0,W=1),Translation=(X=0,Y=-6.5653,Z=0),Scale3D=(X=1,Y=1,Z=1))),(Item=(Type=Bone,Name=\"thigh_corr_scale_r\"),Transform=(Rotation=(X=0,Y=0,Z=0,W=1),Translation=(X=0,Y=3.3339,Z=70.111),Scale3D= ...(truncated)"
    Weight=@n42.OutputParam
   )
   n43 = &RigUnit_SphericalPoseReader(
    ActiveRegionScaleFactors=(PositiveWidth=1,NegativeWidth=1,PositiveHeight=1,NegativeHeight=1)
    ActiveRegionSize=0.1
    Debug=(bDrawDebug=false,bDraw2D=False,bDrawLocalAxes=False,DebugScale=25,DebugSegments=20,DebugThickness=0.25)
    DriverAxis=(X=-1,Y=0,Z=0)
    DriverItem=(Type=Bone,Name="upperarm_r")
    FalloffRegionScaleFactors=(PositiveWidth=1,NegativeWidth=1,PositiveHeight=1,NegativeHeight=1)
    FalloffSize=0.2
    FlipHeightScaling=False
    FlipWidthScaling=False
    OptionalParentItem=(Type=Bone,Name="None")
    RotationOffset=(X=130,Y=0,Z=90)
    ExecutePin=@RigVMFunction_Sequence_1.D
   )
   > n44 = &RigUnit_SphericalPoseReader(
    ActiveRegionScaleFactors=(PositiveWidth=1,NegativeWidth=1,PositiveHeight=1,NegativeHeight=1)
    ActiveRegionSize=0.1
    Debug=(bDrawDebug=false,bDraw2D=False,bDrawLocalAxes=False,DebugScale=25,DebugSegments=20,DebugThickness=0.25)
    DriverAxis=(X=1,Y=0,Z=0)
    DriverItem=(Type=Bone,Name="upperarm_l")
    FalloffRegionScaleFactors=(PositiveWidth=1,NegativeWidth=1,PositiveHeight=1,NegativeHeight=1)
    FalloffSize=0.2
    FlipHeightScaling=False
    FlipWidthScaling=False
    OptionalParentItem=(Type=Bone,Name="None")
    RotationOffset=(X=60,Y=0,Z=-90)
   )
   > &RigUnit_ModifyTransforms(
    ItemToModify=((Item=(Type=Bone,Name="dyn_master_sleeve"),Transform=(Rotation=(X=0,Y=0,Z=-0.0436,W=0.999),Translation=(X=0,Y=0.7026,Z=0),Scale3D=(X=1,Y=1,Z=1))))
    WeightMinimum=0
    Weight=&RigVMFunction_MathDoubleMul(B=@n43.OutputParam, A=@n44.OutputParam).Result
   )
   n45 = &RigUnit_GetRelativeTransformForItem(Child=(Type=Bone,Name="foot_l"), Parent=(Type=Bone,Name="calf_l"), bChildInitial=False, bParentInitial=False)
   &RigUnit_ModifyTransforms(
    ItemToModify=((Item=(Type=Bone,Name="heel_corr_l"),Transform=(Rotation=(X=0,Y=0,Z=0,W=1),Translation=(X=4,Y=-4,Z=0),Scale3D=(X=1,Y=1,Z=1))))
    WeightMinimum=0
    Weight=&RigVMFunction_MathDoubleRemap(
     SourceMaximum=50
     SourceMinimum=30
     TargetMaximum=1
     TargetMinimum=0
     Value=&RigVMFunction_MathQuaternionToEuler(RotationOrder=ZYX, Value=@n45.RelativeTransform.Rotation).Result.Z
    ).Result
    ExecutePin=@RigVMFunction_Sequence_1.E
   )
   n46 = &RigUnit_GetRelativeTransformForItem(Child=(Type=Bone,Name="foot_r"), Parent=(Type=Bone,Name="calf_r"), bChildInitial=False, bParentInitial=False)
   > &RigUnit_ModifyTransforms(
    ItemToModify=((Item=(Type=Bone,Name="heel_corr_r"),Transform=(Rotation=(X=0,Y=0,Z=0,W=1),Translation=(X=-4,Y=4,Z=0),Scale3D=(X=1,Y=1,Z=1))))
    WeightMinimum=0
    Weight=&RigVMFunction_MathDoubleRemap(
     SourceMaximum=50
     SourceMinimum=1
     TargetMaximum=1
     TargetMinimum=0
     Value=&RigVMFunction_MathQuaternionToEuler(RotationOrder=ZYX, Value=@n46.RelativeTransform.Rotation).Result.Z
    ).Result
   )
   &RigUnit_ModifyTransforms(
    ItemToModify="((Item=(Type=Bone,Name=\"Base_back\"),Transform=(Rotation=(X=-0.0436,Y=0,Z=0,W=0.999),Translation=(X=0,Y=-5,Z=-1.5),Scale3D=(X=1,Y=1,Z=1))),(Item=(Type=Bone,Name=\"Base_thingCorr_r\"),Transform=(Rotation=(X=0,Y=0.0436,Z=0,W=0.999),Translation=(X=0,Y=0,Z=-3),Sc ...(truncated)"
    Weight=1
    WeightMinimum=0
    ExecutePin=@n10.True
   )
   n47 = &RigUnit_GetRelativeTransformForItem(Child=(Type=Bone,Name="foot_l"), Parent=(Type=Bone,Name="calf_l"), bChildInitial=False, bParentInitial=False)
   > &RigUnit_ModifyTransforms(
    ItemToModify="((Item=(Type=Bone,Name=\"Base_back\"),Transform=(Rotation=(X=-0.0495,Y=0,Z=0,W=0.9988),Translation=(X=0,Y=-13.0179,Z=4),Scale3D=(X=1,Y=1,Z=1))),(Item=(Type=Bone,Name=\"Base_thingCorr_r\"),Transform=(Rotation=(X=0,Y=0.0175,Z=0,W=0.9998),Translation=(X=-3.6273,Y ...(truncated)"
    WeightMinimum=0
    Weight=&RigVMFunction_MathDoubleRemap(
     SourceMaximum=50
     SourceMinimum=30
     TargetMaximum=1
     TargetMinimum=0
     Value=&RigVMFunction_MathQuaternionToEuler(RotationOrder=ZYX, Value=@n47.RelativeTransform.Rotation).Result.Z
    ).Result
   )
   > n48 = &RigUnit_SphericalPoseReader(
    ActiveRegionScaleFactors=(PositiveWidth=1,NegativeWidth=4,PositiveHeight=1.5,NegativeHeight=1.5)
    ActiveRegionSize=0.05
    Debug=(bDrawDebug=false,DebugScale=25,DebugSegments=20,DebugThickness=0.25)
    DriverAxis=(X=1,Y=0,Z=0)
    DriverItem=(Type=Bone,Name="thigh_r")
    FalloffRegionScaleFactors=(PositiveWidth=1,NegativeWidth=1,PositiveHeight=1,NegativeHeight=1)
    FalloffSize=0.4
    FlipHeightScaling=false
    OptionalParentItem=(Type=Bone)
    RotationOffset=(X=-20,Y=15,Z=0)
   )
   > n49 = &RigUnit_SphericalPoseReader(
    ActiveRegionScaleFactors=(PositiveWidth=4,NegativeWidth=1,PositiveHeight=1.5,NegativeHeight=1.5)
    ActiveRegionSize=0.05
    Debug=(bDrawDebug=false,DebugScale=25,DebugSegments=20,DebugThickness=0.25)
    DriverAxis=(X=-1,Y=0,Z=0)
    DriverItem=(Type=Bone,Name="thigh_l")
    FalloffRegionScaleFactors=(PositiveWidth=1,NegativeWidth=1,PositiveHeight=1,NegativeHeight=1)
    FalloffSize=0.4
    FlipHeightScaling=false
    OptionalParentItem=(Type=Bone)
    RotationOffset=(X=170,Y=15,Z=0)
   )
   > &RigUnit_ModifyTransforms(
    ItemToModify="((Item=(Type=Bone,Name=\"base_sleeve_l\"),Transform=(Rotation=(X=0,Y=-0.4915,Z=0,W=0.8709),Translation=(X=-3.9713,Y=0,Z=1.9764),Scale3D=(X=1,Y=1,Z=1))),(Item=(Type=Bone,Name=\"Base_back\"),Transform=(Rotation=(X=0,Y=0,Z=0,W=1),Translation=(X=-1.4391,Y=0,Z=0),S ...(truncated)"
    Weight=&RigVMFunction_MathDoubleMul(A=@n49.OutputParam, B=@n48.OutputParam).Result
   )
   > n50 = &RigUnit_SphericalPoseReader(
    ActiveRegionScaleFactors=(PositiveWidth=3,NegativeWidth=1,PositiveHeight=3,NegativeHeight=3)
    ActiveRegionSize=0.05
    Debug=(bDrawDebug=false,DebugScale=25,DebugSegments=20,DebugThickness=0.25)
    DriverAxis=(X=-1,Y=0,Z=0)
    DriverItem=(Type=Bone,Name="thigh_l")
    FalloffRegionScaleFactors=(PositiveWidth=1,NegativeWidth=1,PositiveHeight=1,NegativeHeight=1)
    FalloffSize=0.3
    FlipHeightScaling=false
    OptionalParentItem=(Type=Bone)
    RotationOffset=(X=104.0915,Y=-17.2262,Z=-11.7722)
   )
   > &RigUnit_ModifyTransforms(
    ItemToModify="((Item=(Type=Bone,Name=\"Base_back\"),Transform=(Rotation=(X=-0.3135,Y=0,Z=0,W=0.9496),Translation=(X=2.7295,Y=-3.7819,Z=0),Scale3D=(X=1,Y=1,Z=1))),(Item=(Type=Bone,Name=\"Base_thingCorr_l\"),Transform=(Rotation=(X=0,Y=0,Z=0,W=1),Translation=(X=0,Y=0,Z=0),Scal ...(truncated)"
    WeightMinimum=0
    Weight=@n50.OutputParam
   )
   > n51 = &RigUnit_SphericalPoseReader(
    ActiveRegionScaleFactors=(PositiveWidth=0.8,NegativeWidth=0,PositiveHeight=0.2,NegativeHeight=0.5)
    ActiveRegionSize=0.3
    Debug=(bDrawDebug=true,bDraw2D=False,bDrawLocalAxes=False,DebugScale=25,DebugSegments=20,DebugThickness=0.25)
    DriverAxis=(X=1,Y=0,Z=0)
    DriverItem=(Type=Bone,Name="thigh_r")
    FalloffRegionScaleFactors=(PositiveWidth=0.4,NegativeWidth=0,PositiveHeight=0.4,NegativeHeight=0.4)
    FalloffSize=0.5
    FlipHeightScaling=False
    FlipWidthScaling=False
    OptionalParentItem=(Type=Bone,Name="pelvis")
    RotationOffset=(X=-90,Y=0,Z=-60)
   )
   > &RigUnit_ModifyTransforms(
    ItemToModify="((Item=(Type=Bone,Name=\"dyn_master_sleeve\"),Transform=(Rotation=(X=0.0411,Y=0.2354,Z=-0.01,W=0.971),Translation=(X=9.2521,Y=0,Z=-3.8204),Scale3D=(X=1,Y=1,Z=1))),(Item=(Type=Bone,Name=\"base_sleeve_r\"),Transform=(Rotation=(X=0,Y=0,Z=0,W=1),Translation=(X=0,Y ...(truncated)"
    WeightMinimum=0
    Weight=@n51.OutputParam
   )
   > Get Delta Ref_3 = RigVMFunctionReferenceNode(Item=(Type=Bone,Name="thigh_r"))
   > &RigUnit_ModifyTransforms(
    ItemToModify="((Item=(Type=Bone,Name=\"Base_thingCorr_r\"),Transform=(Rotation=(X=0,Y=0,Z=0,W=1),Translation=(X=-0.163,Y=9.1032,Z=0.0041),Scale3D=(X=1,Y=1,Z=1))),(Item=(Type=Bone,Name=\"Base_thingCorr_l\"),Transform=(Rotation=(X=0,Y=0,Z=0,W=1),Translation=(X=-3.2187,Y=-5.19 ...(truncated)"
    WeightMinimum=0
    Weight=&RigVMFunction_MathDoubleRemap(SourceMaximum=35, SourceMinimum=15, TargetMaximum=1, TargetMinimum=0, Value=@"Get Delta Ref_3".Result.Rotation.Roll).Result
   )
   n52 = &RigUnit_SphericalPoseReader(
    ActiveRegionScaleFactors=(PositiveWidth=1,NegativeWidth=1,PositiveHeight=1,NegativeHeight=1)
    ActiveRegionSize=0.1
    Debug=(bDrawDebug=false,bDraw2D=False,bDrawLocalAxes=False,DebugScale=25,DebugSegments=20,DebugThickness=0.25)
    DriverAxis=(X=1,Y=0,Z=0)
    DriverItem=(Type=Bone,Name="thigh_r")
    FalloffRegionScaleFactors=(PositiveWidth=1,NegativeWidth=1,PositiveHeight=1,NegativeHeight=1)
    FalloffSize=0.2
    FlipHeightScaling=False
    FlipWidthScaling=False
    OptionalParentItem=(Type=Bone,Name="None")
    RotationOffset=(X=50,Y=0,Z=-120)
    ExecutePin=@RigVMFunction_Sequence_2_1.H
   )
   > n53 = &RigUnit_SphericalPoseReader(
    ActiveRegionScaleFactors=(PositiveWidth=1,NegativeWidth=1,PositiveHeight=1,NegativeHeight=1)
    ActiveRegionSize=0.1
    Debug=(bDrawDebug=false,bDraw2D=False,bDrawLocalAxes=False,DebugScale=25,DebugSegments=20,DebugThickness=0.25)
    DriverAxis=(X=-1,Y=0,Z=0)
    DriverItem=(Type=Bone,Name="thigh_l")
    FalloffRegionScaleFactors=(PositiveWidth=1,NegativeWidth=1,PositiveHeight=1,NegativeHeight=1)
    FalloffSize=0.2
    FlipHeightScaling=False
    FlipWidthScaling=False
    OptionalParentItem=(Type=Bone,Name="None")
    RotationOffset=(X=90,Y=0,Z=50)
   )
   Multiply_1_1_1_1_1_3 = RigVMAggregateNode(
    B=&RigVMFunction_MathDoubleRemap(
     SourceMaximum=30
     SourceMinimum=35
     TargetMaximum=1
     TargetMinimum=0
     Value=&RigUnit_GetRelativeTransformForItem(Child=(Type=Bone,Name="pelvis"), Parent=(Type=Bone,Name="root"), bChildInitial=False, bParentInitial=False).RelativeTransform.Translation.Z
    ).Result
    A=@n52.OutputParam
    C=@n53.OutputParam
   )
   > &RigUnit_ModifyTransforms(
    ItemToModify=((Item=(Type=Bone,Name="base_sleeve_r"),Transform=(Rotation=(X=0,Y=-0.2424,Z=0,W=0.9702),Translation=(X=0,Y=0,Z=0),Scale3D=(X=1,Y=1,Z=1))))
    WeightMinimum=0
    Weight=@Multiply_1_1_1_1_1_3.Result
   )
   n54 = &RigVMFunction_ControlFlowBranch(ExecuteContext=@RigVMFunction_Sequence_2_1.I, Condition=RigVMVariableNode() {variable=isLOD2}.Value)
   &RigUnit_ModifyTransforms(
    ItemToModify="((Item=(Type=Bone,Name=\"Base_back\"),Transform=(Rotation=(X=0,Y=0,Z=0,W=1),Translation=(X=0,Y=-5.5,Z=0),Scale3D=(X=1,Y=1,Z=1))),(Item=(Type=Bone,Name=\"Base_thingCorr_r\"),Transform=(Rotation=(X=0,Y=0,Z=0,W=1),Translation=(X=0,Y=0,Z=-3),Scale3D=(X=1,Y=1,Z=1)) ...(truncated)"
    Weight=1
    WeightMinimum=0
    ExecutePin=@n54.True
   )
   n55 = &RigVMFunction_ControlFlowBranch(ExecuteContext=@RigVMFunction_Sequence_2_1.J, Condition=RigVMVariableNode() {variable=isDBNO}.Value)
   &RigUnit_ModifyTransforms(
    ItemToModify=((Item=(Type=Bone,Name="skirt_base"),Transform=(Rotation=(X=0,Y=0,Z=-0.0793,W=0.9969),Translation=(X=0,Y=0,Z=0),Scale3D=(X=1,Y=1,Z=1))))
    Weight=1
    WeightMinimum=0
    ExecutePin=@n55.True
   )
   n56 = &RigUnit_SphericalPoseReader(
    ActiveRegionScaleFactors=(PositiveWidth=1,NegativeWidth=1,PositiveHeight=1,NegativeHeight=1)
    ActiveRegionSize=0.05
    Debug=(bDrawDebug=false,DebugScale=25,DebugSegments=20,DebugThickness=0.25)
    DriverAxis=(X=1,Y=0,Z=0)
    DriverItem=(Type=Bone,Name="thigh_r")
    ExecutePin=()
    FalloffRegionScaleFactors=(PositiveWidth=1,NegativeWidth=1,PositiveHeight=1,NegativeHeight=1)
    FalloffSize=0.1
    OptionalParentItem=(Type=Bone)
    RotationOffset=(X=-107,Y=0,Z=90)
   )
   > &RigUnit_ModifyTransforms(
    ItemToModify="((Item=(Type=Bone,Name=\"Base_back\"),Transform=(Rotation=(X=0.015,Y=-0.1823,Z=0.0836,W=0.9796),Translation=(X=3.2984,Y=0.0842,Z=0.2515),Scale3D=(X=1,Y=1,Z=1))),(Item=(Type=Bone,Name=\"Base_thingCorr_l\"),Transform=(Rotation=(X=0,Y=0,Z=0,W=1),Translation=(X=0, ...(truncated)"
    Weight=@n56.OutputParam
   )
   &RigUnit_ModifyTransforms(
    ExecutePin=()
    ItemToModify="((Item=(Type=Bone,Name=\"Base_back\"),Transform=(Rotation=(X=-0.0872,Y=0,Z=0,W=0.9962),Translation=(X=-9.1698,Y=-3.6286,Z=0.2872),Scale3D=(X=1,Y=1,Z=1))),(Item=(Type=Bone,Name=\"thigh_corr_scale_r\"),Transform=(Rotation=(X=0,Y=0,Z=0,W=1),Translation=(X=0,Y=3.3 ...(truncated)"
    Weight=1
   )
  ]}
]
hierarchy [
 root {loc=<0 0 0>, rot=<0 0 0 1>}
 pelvis {parent=root, loc=<0 -0.5301 98.0571>, rot=<0.0127 0.707 -0.0127 -0.707>}
 spine_01 {parent=pelvis, loc=<6.1208 -0.9328 0>, rot=<0 0 0.034 -0.9994>}
 spine_02 {parent=spine_01, loc=<6.5997 -0.2095 0>, rot=<0 0 0.0238 -0.9997>}
 spine_03 {parent=spine_02, loc=<6.4914 -0.1237 0>, rot=<0 0 0.0149 -0.9999>}
 spine_04 {parent=spine_03, loc=<6.9598 0.0794 0>, rot=<0 0 -0.0132 -0.9999>}
 spine_05 {parent=spine_04, loc=<6.9987 0 0>, rot=<0 0 -0.0015 -1>}
 attach_cape {parent=spine_05, loc=<9.3216 -8.4501 0>, rot=<0.0616 -0.7044 -0.0616 -0.7044>}
 neck_01 {parent=spine_05, loc=<11.3138 0.1529 0>, rot=<0 0 -0.2003 -0.9797>}
 neck_02 {parent=neck_01, loc=<4.5627 0 0>, rot=<0 0 0.0504 -0.9987>}
 head {parent=neck_02, loc=<5.3914 0 0>, rot=<0 0 0.115 -0.9934>}
 attach_backpack {parent=spine_05, loc=<-0.1237 -13.4564 0.0006>, rot=<-0.0093 -0.707 0.0093 -0.707>}
 clavicle_l {parent=spine_05, loc=<5.4831 -0.1053 -3.2946>, rot=<0.0465 -0.7289 -0.009 -0.6829>}
 upperarm_l {parent=clavicle_l, loc=<9.8658 -1.2136 -0.4105>, rot=<0.0164 -0.3245 -0.0269 -0.9454>}
 lowerarm_l {parent=upperarm_l, loc=<26.465 -0.8345 0.259>, rot=<0.0036 0.0161 -0.2298 -0.9731>}
 hand_l {parent=lowerarm_l, loc=<22.7355 -0.3203 -0.6105>, rot=<-0.6784 -0.0223 0.0735 -0.7307>}
 thumb_01_l {parent=hand_l, loc=<1.3118 -0.3191 -2.459>, rot=<0.6546 -0.4042 0.0664 -0.6354>}
 thumb_02_l {parent=thumb_01_l, loc=<5.9336 -0.0448 0.0401>, rot=<-0.0037 0.0213 0.298 -0.9543>}
 thumb_03_l {parent=thumb_02_l, loc=<2.2223 0.3682 -0.0907>, rot=<0.0118 -0.013 0.0066 -0.9998>}
 index_metacarpal_l {parent=hand_l, loc=<1.6535 0.0183 -2.96>, rot=<0.0158 -0.0124 -0.0173 -0.9996>}
 index_01_l {parent=index_metacarpal_l, loc=<8.3008 0.0357 -0.0524>, rot=<-0.0044 0.0485 0.0924 -0.9945>}
 index_02_l {parent=index_01_l, loc=<3.3587 -0.3004 0.0452>, rot=<-0.002 0.0027 0.3225 -0.9466>}
 index_03_l {parent=index_02_l, loc=<2.5578 -0.123 -0.0252>, rot=<-0.002 0.0067 0.2558 -0.9667>}
 middle_metacarpal_l {parent=hand_l, loc=<2.1856 -0.1725 -0.5984>, rot=<0.006 0.0168 -0.0329 -0.9993>}
 middle_01_l {parent=middle_metacarpal_l, loc=<7.3275 -0.2447 0.0129>, rot=<-0.0208 0.0042 0.0858 -0.9961>}
 middle_02_l {parent=middle_01_l, loc=<4.0123 0.1401 0.1175>, rot=<-0.0218 0.0258 0.3512 -0.9357>}
 middle_03_l {parent=middle_02_l, loc=<2.6077 0.0534 0.1373>, rot=<0.0082 0.0188 0.1708 -0.9851>}
 ring_metacarpal_l {parent=hand_l, loc=<2.2074 -0.3625 1.3638>, rot=<-0.0284 0.0722 -0.0458 -0.9959>}
 ring_01_l {parent=ring_metacarpal_l, loc=<7.0295 -0.6197 -0.1446>, rot=<-0.0489 -0.0252 0.1892 -0.9804>}
 ring_02_l {parent=ring_01_l, loc=<3.6174 0.4647 -0.1027>, rot=<0 0 0.2806 -0.9598>}
 ring_03_l {parent=ring_02_l, loc=<2.4152 0.114 0.0199>, rot=<0 0 0.1189 -0.9929>}
 pinky_metacarpal_l {parent=hand_l, loc=<1.7616 -0.9753 3.3814>, rot=<-0.0397 0.0814 -0.069 -0.9935>}
 pinky_01_l {parent=pinky_metacarpal_l, loc=<7.1709 -0.4831 -0.1094>, rot=<-0.0654 -0.0035 0.2242 -0.9723>}
 pinky_02_l {parent=pinky_01_l, loc=<2.8274 0.0893 -0.1043>, rot=<0.0037 0.0011 0.2854 -0.9584>}
 pinky_03_l {parent=pinky_02_l, loc=<2.2046 0.0307 -0.0078>, rot=<-0.0002 0.0098 0.1445 -0.9895>}
 deform_wrist_dn_base_l {parent=hand_l, loc=<0 -1.5326 0>, rot=<0 0 0 -1>}
 deform_wrist_dn_l {parent=deform_wrist_dn_base_l, loc=<0 0 0>, rot=<0 0 0 1>}
 deform_wrist_up_base_l {parent=hand_l, loc=<0 2.3009 0>, rot=<0 0 0 -1>}
 deform_wrist_up_l {parent=deform_wrist_up_base_l, loc=<0 0 0>, rot=<0 0 0 1>}
 lowerarm_twist_02_l {parent=lowerarm_l, loc=<10.6707 0 0>, rot=<0 0 0 1>}
 lowerarm_twist_01_l {parent=lowerarm_l, loc=<18.316 0 0>, rot=<0 0 0 1>}
 deform_elbow_in_base_l {parent=lowerarm_l, loc=<0.1834 3.6062 -0.2572>, rot=<0 0 0 -1>}
 deform_elbow_in_l {parent=deform_elbow_in_base_l, loc=<0 0 0>, rot=<0 0 0 1>}
 deform_elbow_out_base_l {parent=lowerarm_l, loc=<0 -4 0>, rot=<0 0 0 1>}
 deform_elbow_out_l {parent=deform_elbow_out_base_l, loc=<0 0 0>, rot=<0 0 0 1>}
 upperarm_twist_01_l {parent=upperarm_l, loc=<7.2135 -0.1134 0>, rot=<0 0 0 1>}
 deform_upperarm_bk_base_l {parent=upperarm_twist_01_l, loc=<-6.4053 -4.3746 0.314>, rot=<0 0 0 -1>}
 deform_upperarm_bk_l {parent=deform_upperarm_bk_base_l, loc=<0 0 0>, rot=<0 0 0 1>}
 deform_upperarm_fr_base_l {parent=upperarm_twist_01_l, loc=<-5.6312 5.092 0.3354>, rot=<0 0 0 -1>}
 deform_upperarm_fr_l {parent=deform_upperarm_fr_base_l, loc=<0 0 0>, rot=<0 0 0 1>}
 deform_upperarm_up_base_l {parent=upperarm_twist_01_l, loc=<-8.891 0.4825 3.9517>, rot=<0 0 0 -1>}
 deform_upperarm_up_l {parent=deform_upperarm_up_base_l, loc=<0 0 0>, rot=<0 0 0 1>}
 upperarm_twist_02_l {parent=upperarm_l, loc=<19.7998 -0.1134 0>, rot=<0 0 0 1>}
 clavicle_r {parent=spine_05, loc=<5.4831 -0.1053 3.2946>, rot=<0.7289 0.0465 -0.6829 0.009>}
 upperarm_r {parent=clavicle_r, loc=<-9.8658 1.2136 0.4105>, rot=<0.0164 -0.3245 -0.0269 -0.9454>}
 lowerarm_r {parent=upperarm_r, loc=<-26.465 0.8345 -0.259>, rot=<0.0036 0.0161 -0.2298 -0.9731>}
 hand_r {parent=lowerarm_r, loc=<-22.7355 0.3203 0.6106>, rot=<-0.6784 -0.0223 0.0735 -0.7307>}
 thumb_01_r {parent=hand_r, loc=<-1.3118 0.3191 2.459>, rot=<0.6546 -0.4042 0.0664 -0.6354>}
 thumb_02_r {parent=thumb_01_r, loc=<-5.9336 0.0448 -0.0401>, rot=<-0.0037 0.0213 0.298 -0.9543>}
 thumb_03_r {parent=thumb_02_r, loc=<-2.2223 -0.3682 0.0907>, rot=<0.0118 -0.013 0.0066 -0.9998>}
 index_metacarpal_r {parent=hand_r, loc=<-1.6535 -0.0183 2.96>, rot=<0.0158 -0.0124 -0.0173 -0.9996>}
 index_01_r {parent=index_metacarpal_r, loc=<-8.3008 -0.0357 0.0524>, rot=<-0.0044 0.0485 0.0924 -0.9945>}
 index_02_r {parent=index_01_r, loc=<-3.3587 0.3004 -0.0452>, rot=<-0.002 0.0027 0.3225 -0.9466>}
 index_03_r {parent=index_02_r, loc=<-2.5578 0.123 0.0252>, rot=<-0.002 0.0067 0.2558 -0.9667>}
 middle_metacarpal_r {parent=hand_r, loc=<-2.1856 0.1725 0.5984>, rot=<0.006 0.0168 -0.0329 -0.9993>}
 middle_01_r {parent=middle_metacarpal_r, loc=<-7.3275 0.2447 -0.0129>, rot=<-0.0208 0.0042 0.0858 -0.9961>}
 middle_02_r {parent=middle_01_r, loc=<-4.0123 -0.1401 -0.1175>, rot=<-0.0218 0.0258 0.3512 -0.9357>}
 middle_03_r {parent=middle_02_r, loc=<-2.6077 -0.0534 -0.1373>, rot=<0.0082 0.0188 0.1708 -0.9851>}
 ring_metacarpal_r {parent=hand_r, loc=<-2.2074 0.3625 -1.3638>, rot=<-0.0284 0.0722 -0.0458 -0.9959>}
 ring_01_r {parent=ring_metacarpal_r, loc=<-7.0295 0.6197 0.1446>, rot=<-0.0489 -0.0252 0.1892 -0.9804>}
 ring_02_r {parent=ring_01_r, loc=<-3.6174 -0.4647 0.1027>, rot=<0 0 0.2806 -0.9598>}
 ring_03_r {parent=ring_02_r, loc=<-2.4152 -0.114 -0.0199>, rot=<0 0 0.1189 -0.9929>}
 pinky_metacarpal_r {parent=hand_r, loc=<-1.7616 0.9753 -3.3814>, rot=<-0.0397 0.0814 -0.069 -0.9935>}
 pinky_01_r {parent=pinky_metacarpal_r, loc=<-7.1709 0.4831 0.1094>, rot=<-0.0654 -0.0035 0.2242 -0.9723>}
 pinky_02_r {parent=pinky_01_r, loc=<-2.8274 -0.0893 0.1043>, rot=<0.0037 0.0011 0.2854 -0.9584>}
 pinky_03_r {parent=pinky_02_r, loc=<-2.2046 -0.0307 0.0078>, rot=<-0.0002 0.0098 0.1445 -0.9895>}
 deform_wrist_dn_base_r {parent=hand_r, loc=<-0.0001 1.5328 0>, rot=<0 0 0 -1>}
 deform_wrist_dn_r {parent=deform_wrist_dn_base_r, loc=<0 0 0>, rot=<0 0 0 1>}
 deform_wrist_up_base_r {parent=hand_r, loc=<-0.0002 -2.3007 -0.0001>, rot=<0 0 0 -1>}
 deform_wrist_up_r {parent=deform_wrist_up_base_r, loc=<0 0 0>, rot=<0 0 0 1>}
 lowerarm_twist_02_r {parent=lowerarm_r, loc=<-10.6707 0 0>, rot=<0 0 0 -1>}
 lowerarm_twist_01_r {parent=lowerarm_r, loc=<-18.316 0 0>, rot=<0 0 0 -1>}
 deform_elbow_in_base_r {parent=lowerarm_r, loc=<-0.1834 -3.6062 0.2572>, rot=<0 0 0 -1>}
 deform_elbow_in_r {parent=deform_elbow_in_base_r, loc=<0 0 0>, rot=<0 0 0 1>}
 deform_elbow_out_base_r {parent=lowerarm_r, loc=<0 4 -0.0001>, rot=<0 0 0 -1>}
 deform_elbow_out_r {parent=deform_elbow_out_base_r, loc=<0 0 0>, rot=<0 0 0 1>}
 upperarm_twist_01_r {parent=upperarm_r, loc=<-7.2135 0.1134 0>, rot=<0 0 0 -1>}
 deform_upperarm_bk_base_r {parent=upperarm_twist_01_r, loc=<6.4053 4.3746 -0.314>, rot=<0 0 0 -1>}
 deform_upperarm_bk_r {parent=deform_upperarm_bk_base_r, loc=<0 0 0>, rot=<0 0 0 1>}
 deform_upperarm_fr_base_r {parent=upperarm_twist_01_r, loc=<5.6312 -5.092 -0.3354>, rot=<0 0 0 -1>}
 deform_upperarm_fr_r {parent=deform_upperarm_fr_base_r, loc=<0 0 0>, rot=<0 0 0 1>}
 deform_upperarm_up_base_r {parent=upperarm_twist_01_r, loc=<8.891 -0.4825 -3.9517>, rot=<0 0 0 -1>}
 deform_upperarm_up_r {parent=deform_upperarm_up_base_r, loc=<0 0 0>, rot=<0 0 0 1>}
 upperarm_twist_02_r {parent=upperarm_r, loc=<-19.7998 0.1134 0>, rot=<0 0 0 -1>}
 deform_pec_aim_base_l {parent=spine_04, loc=<3.2456 9.3309 -1.2237>, rot=<0.0603 -0.4033 0.3318 -0.8507>}
 deform_pec_aim_l {parent=deform_pec_aim_base_l, loc=<0 0 0>, rot=<0 0 0 1>}
 deform_pec_base_l {parent=deform_pec_aim_l, loc=<9.3665 0 0>, rot=<0 0 0 1>}
 deform_pec_l {parent=deform_pec_base_l, loc=<0 0 0>, rot=<0 0 0 1>}
 deform_pec_aim_base_r {parent=spine_04, loc=<3.2452 9.3309 1.2237>, rot=<-0.4033 -0.0603 0.8507 0.3318>}
 deform_pec_aim_r {parent=deform_pec_aim_base_r, loc=<0 0 0>, rot=<0 0 0 1>}
 deform_pec_base_r {parent=deform_pec_aim_r, loc=<-9.3667 -0.0003 -0.0004>, rot=<0 0 0 1>}
 deform_pec_r {parent=deform_pec_base_r, loc=<0 0 0>, rot=<0 0 0 1>}
 thigh_l {parent=pelvis, loc=<-1.0855 -0.2109 -8.7053>, rot=<0.0009 0.0355 0.0155 -0.9993>}
 calf_l {parent=thigh_l, loc=<-42.5284 -0.0019 -0.7297>, rot=<0.0006 -0.0161 -0.0391 -0.9991>}
 foot_l {parent=calf_l, loc=<-46.7401 -2.5014 -0.592>, rot=<0.0427 -0.0238 0.0353 -0.9982>}
 ball_l {parent=foot_l, loc=<-3.5658 10.8066 -0.1544>, rot=<0.0001 -0.0001 -0.7063 -0.7079>}
 calf_twist_02_l {parent=calf_l, loc=<-19.1546 -1.7943 -0.2638>, rot=<0.0237 0.0095 -0.0035 -0.9997>}
 calf_twist_01_l {parent=calf_l, loc=<-35.8483 -1.9197 -0.5794>, rot=<0.0237 0.0095 -0.0035 -0.9997>}
 deform_knee_in_base_l {parent=calf_l, loc=<0.8031 -5.0231 0.2456>, rot=<0 0 0 -1>}
 deform_knee_in_l {parent=deform_knee_in_base_l, loc=<0 0 0>, rot=<0 0 0 1>}
 deform_knee_out_base_l {parent=calf_l, loc=<0 4 0>, rot=<0 0 0 -1>}
 deform_knee_out_l {parent=deform_knee_out_base_l, loc=<0 0 0>, rot=<0 0 0 1>}
 calf_l_null {parent=calf_l, type=Null, loc=<0 0 0>, rot=<0 0 0 1>}
 base_heel_corr_l {parent=calf_l, loc=<-43.3822 -5.6137 -0.5013>, rot=<0.0427 -0.0238 0.0353 -0.9982>}
 heel_corr_l {parent=base_heel_corr_l, loc=<0 0 0>, rot=<0 0 0 1>}
 thigh_twist_01_l {parent=thigh_l, loc=<-27.952 0 0>, rot=<0.0013 0 0 -1>}
 deform_groin_base_l {parent=thigh_l, loc=<0.4801 5.6776 -0.2486>, rot=<0 0 0 -1>}
 deform_groin_l {parent=deform_groin_base_l, loc=<0 0 0>, rot=<0 0 0 1>}
 thigh_corr_l {parent=thigh_l, loc=<-14.7289 -12.1376 0.048>, rot=<0.0013 0 0 -1>}
 thigh_corr_scale_l {parent=thigh_l, loc=<-12.9681 0.0424 0.6265>, rot=<0.0013 0 0 -1>}
 Base_thingCorr_l {parent=thigh_l, loc=<-12.952 -13.2156 -0.0195>, rot=<0.031 0.5484 0.8355 -0.0175>}
 dyn_thingCorr_l {parent=Base_thingCorr_l, loc=<0 0 0>, rot=<0 0 0 1>}
 master_thigh_corSide_l {parent=thigh_l, loc=<-4.5506 -1.4226 -7.7277>, rot=<0 -0.9982 0 -0.0592>}
 thigh_corSide_l {parent=master_thigh_corSide_l, loc=<0 0 0>, rot=<0 0 0 1>}
 thigh_corSide_l_01 {parent=thigh_corSide_l, loc=<7 0 0>, rot=<0 0 0 1>}
 thigh_r {parent=pelvis, loc=<-1.0855 -0.2109 8.7053>, rot=<0.0355 -0.0009 0.9993 0.0155>}
 calf_r {parent=thigh_r, loc=<42.5284 0.0019 0.7297>, rot=<0.0006 -0.0161 -0.0391 -0.9991>}
 foot_r {parent=calf_r, loc=<46.7401 2.5014 0.592>, rot=<0.0427 -0.0238 0.0353 -0.9982>}
 ball_r {parent=foot_r, loc=<3.5658 -10.8066 0.1544>, rot=<0.0001 -0.0001 -0.7063 -0.7079>}
 calf_twist_02_r {parent=calf_r, loc=<19.1546 1.7943 0.2638>, rot=<0.0237 0.0095 -0.0035 -0.9997>}
 calf_twist_01_r {parent=calf_r, loc=<35.8483 1.9197 0.5794>, rot=<0.0237 0.0095 -0.0035 -0.9997>}
 deform_knee_in_base_r {parent=calf_r, loc=<-0.8031 5.0231 -0.2457>, rot=<0 0 0 -1>}
 deform_knee_in_r {parent=deform_knee_in_base_r, loc=<0 0 0>, rot=<0 0 0 1>}
 deform_knee_out_base_r {parent=calf_r, loc=<0 -4 0>, rot=<0 0 0 -1>}
 deform_knee_out_r {parent=deform_knee_out_base_r, loc=<0 0 0>, rot=<0 0 0 1>}
 calf_r_null {parent=calf_r, type=Null, loc=<0 0 0>, rot=<0 0 0 1>}
 base_heel_corr_r {parent=calf_r, loc=<43.5183 5.1743 0.5313>, rot=<0.0427 -0.0238 0.0353 -0.9982>}
 heel_corr_r {parent=base_heel_corr_r, loc=<0 0 0>, rot=<0 0 0 1>}
 thigh_twist_01_r {parent=thigh_r, loc=<27.952 0 0>, rot=<0.0013 0 0 -1>}
 deform_groin_base_r {parent=thigh_r, loc=<-0.4801 -5.6776 0.2486>, rot=<0 0 0 -1>}
 deform_groin_r {parent=deform_groin_base_r, loc=<0 0 0>, rot=<0 0 0 1>}
 thigh_corr_r {parent=thigh_r, loc=<14.7368 12.1375 -0.0478>, rot=<0.0013 0 0 -1>}
 thigh_corr_scale_r {parent=thigh_r, loc=<12.9681 -0.0425 -0.6265>, rot=<0.0013 0 0 -1>}
 Base_thingCorr_r {parent=thigh_r, loc=<12.952 13.2156 0.0195>, rot=<-0.849 -0.0167 0.0314 -0.5272>}
 dyn_thingCorr_r {parent=Base_thingCorr_r, loc=<0 0 0>, rot=<0 0 0 1>}
 master_thigh_corSide_r {parent=thigh_r, loc=<4.5658 1.4237 7.7133>, rot=<-0.0007 0.0755 0.0001 -0.9971>}
 thigh_corSide_r {parent=master_thigh_corSide_r, loc=<0 0 0>, rot=<0 0 0 1>}
 thigh_corSide_r_01 {parent=thigh_corSide_r, loc=<7 0 0>, rot=<0 0 0 1>}
 base_master_sleeve {parent=pelvis, loc=<0 8.4001 0>, rot=<0 0 0 1>}
 dyn_master_sleeve {parent=base_master_sleeve, loc=<0 0 0>, rot=<0 0 0 1>}
 base_sleeve_l {parent=dyn_master_sleeve, loc=<-2.1242 1.1637 -5.2045>, rot=<-0.0047 -0.7945 -0.6028 -0.0734>}
 dyn_sleeve_l_1 {parent=base_sleeve_l, loc=<0 0 0>, rot=<0 0 0 1>}
 dyn_sleeve_l_2 {parent=dyn_sleeve_l_1, loc=<8.6439 0.1076 0>, rot=<0 0 0.017 -0.9999>}
 dyn_sleeve_l_3 {parent=dyn_sleeve_l_2, loc=<8.169 -0.0703 0>, rot=<0 0 -0.0111 -0.9999>}
 dyn_sleeve_l_4 {parent=dyn_sleeve_l_3, loc=<8.9367 0.1306 0>, rot=<0 0 0.0159 -0.9999>}
 base_sleeve_r {parent=dyn_master_sleeve, loc=<-0.693 3.5811 2.5958>, rot=<0.1122 0.611 0.7834 0.0172>}
 dyn_sleeve_r_1 {parent=base_sleeve_r, loc=<0 0 0>, rot=<0 0 0 1>}
 dyn_sleeve_r_2 {parent=dyn_sleeve_r_1, loc=<10.4371 0.5109 -0.3296>, rot=<0.0031 -0.0941 -0.0323 -0.995>}
 dyn_sleeve_r_3 {parent=dyn_sleeve_r_2, loc=<8.202 -0.0696 0>, rot=<0 0 -0.0111 -0.9999>}
 deform_glute_base_l {parent=pelvis, loc=<-7.3466 -9.0213 -9.1245>, rot=<0.0172 0.031 -0.4477 -0.8935>}
 deform_glute_l {parent=deform_glute_base_l, loc=<0 0 0>, rot=<0 0 0 1>}
 deform_glute_base_r {parent=pelvis, loc=<-7.3464 -9.0213 9.1245>, rot=<0.031 -0.0172 0.8935 -0.4477>}
 deform_glute_r {parent=deform_glute_base_r, loc=<0 0 0>, rot=<0 0 0 1>}
 pelvis_scale_corr {parent=pelvis, loc=<-6.4874 -9.6764 0.0488>, rot=<0.0057 -0.705 0.7092 -0.0055>}
 Base_back {parent=pelvis, loc=<-10.0468 -15.1056 0>, rot=<-0.0127 -0.707 0.0127 -0.707>}
 dyn_back {parent=Base_back, loc=<-0.0583 -0.0689 -4.4322>, rot=<0.5 -0.5 -0.5 -0.5>}
 skirt_base {parent=pelvis, loc=<0 0 0>, rot=<0 0 0 1>}
 dyn_skirt_bk_mstr {parent=skirt_base, loc=<-2.1256 -13.2243 -0.0945>, rot=<0.0373 -0.705 0.7064 -0.051>}
 dyn_skirt_bk_01 {parent=dyn_skirt_bk_mstr, loc=<0 0 0>, rot=<0 0 0 1>}
 dyn_skirt_bk_02 {parent=dyn_skirt_bk_01, loc=<8.1866 0.0378 0.0558>, rot=<0.0003 -0.0795 0.0084 -0.9968>}
 dyn_skirt_bk_03 {parent=dyn_skirt_bk_02, loc=<9.2816 0.0201 0.0355>, rot=<0.0004 -0.0077 0.0044 -1>}
 dyn_skirt_bk_r_mstr {parent=skirt_base, loc=<-0.4799 -11.2822 10.3417>, rot=<0.0516 -0.52 0.8523 0.0233>}
 dyn_skirt_bk_r_01 {parent=dyn_skirt_bk_r_mstr, loc=<0 0 0>, rot=<0 0 0 1>}
 dyn_skirt_bk_r_02 {parent=dyn_skirt_bk_r_01, loc=<12.5681 0.0755 -0.3421>, rot=<0.0009 -0.0933 -0.0207 -0.9954>}
 dyn_skirt_bk_l_mstr {parent=skirt_base, loc=<-0.505 -11.3572 -10.3333>, rot=<-0.0257 -0.8456 0.5306 -0.0523>}
 dyn_skirt_bk_l_01 {parent=dyn_skirt_bk_l_mstr, loc=<0 0 0>, rot=<0 0 0 1>}
 dyn_skirt_bk_l_02 {parent=dyn_skirt_bk_l_01, loc=<12.5682 -0.0842 -0.3401>, rot=<0.0033 -0.0928 0.0227 -0.9954>}
]
variables [
 Backward Movement {type=double, default=0}
 Crouch {type=double, default=0}
 Dist-ce Between knee {type=double, default=0}
 Forward Backward Movement {type=double, default=0}
 Forward Movement {type=double, default=0}
 Left Right Movement {type=double, default=0}
 Right Movement {type=double, default=0}
 isAltskin {type=bool, default=False}
 isDBNO {type=bool, default=False}
 isEmote {type=bool, default=False}
 isFloating {type=bool, default=False}
 isLOD2 {type=bool, default=False}
 isSkyDive {type=bool, default=False}
 leftM {type=double, default=0}
]
