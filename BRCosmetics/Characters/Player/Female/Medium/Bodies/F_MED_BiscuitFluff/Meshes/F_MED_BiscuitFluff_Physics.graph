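# Physics asset for the F_MED_BiscuitFluff female medium cosmetic body, giving its dynamic sleeve bones capsule bodies for secondary motion.
# Each sleeve is a chain (four links left, three right) with narrow swing limits and zero twist limit, hung from a base bone locked to the pelvis.
# Capsules on pelvis, lower spine, thighs and thigh twists form body volumes joined by free angular constraints.
# Both base-sleeve-to-pelvis constraints appear twice.
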PhysicsAsset /BRCosmetics/Characters/Player/Female/Medium/Bodies/F_MED_BiscuitFluff/Meshes/F_MED_BiscuitFluff_Physics
bodies {
 base_sleeve_l [{shapeType=Capsule, length=0.505, radius=0.505, shapeName=base_sleeve_l_capsule, rotation=<rot 0 -90 0>}]
 base_sleeve_r [{shapeType=Capsule, length=0.505, radius=0.505, shapeName=base_sleeve_r_capsule, rotation=<rot 0 -90 0>}]
 dyn_sleeve_l_1 [
  {
   shapeType=Capsule
   length=1.2608
   radius=2.454
   shapeName=dyn_sleeve_l_1_capsule
   center=<3.3716 0.3491 1.5855>
   rotation=<rot 0.0234 0.0065 102.417>
  }
  {shapeType=Capsule, length=2.8729, radius=2.1813, center=<7.2607 0.2873 1.3979>, rotation=<rot 0 0 101.6132>}
 ]
 dyn_sleeve_l_2 [
  {
   shapeType=Capsule
   length=4.358
   radius=2.3218
   shapeName=dyn_sleeve_l_2_capsule
   center=<3.3721 0.1178 0.535>
   rotation=<rot 0.0232 0.0065 102.4162>
  }
  {shapeType=Capsule, length=5.1931, radius=2.5856, center=<7.2122 0 0>, rotation=<rot 0 0 96.7543>}
 ]
 dyn_sleeve_l_3 [{shapeType=Capsule, length=4.838, radius=3.2519, shapeName=dyn_sleeve_l_3_capsule, center=<4.8333 0.0721 0.4088>, rotation=<rot 0 0 100>}]
 dyn_sleeve_l_4 [{
   shapeType=Capsule
   length=3.045
   radius=4.0411
   shapeName=dyn_sleeve_l_4_capsule
   center=<3.3723 1.115 -0.2455>
   rotation=<rot 0.0232 0.0065 102.416>
  }]
 dyn_sleeve_r_1 [{
   shapeType=Capsule
   length=3.6995
   radius=2.5505
   shapeName=dyn_sleeve_r_1_capsule
   center=<4.9047 -0.0473 0.5225>
   rotation=<rot 0 0 84.8249>
  }]
 dyn_sleeve_r_2 [
  {shapeType=Capsule, length=4.2451, radius=3.2165, shapeName=dyn_sleeve_r_2_capsule, center=<2.9687 0 0>, rotation=<rot 0 0 84.7853>}
  {
   shapeType=Capsule
   length=4.1818
   radius=2.9117
   shapeName=dyn_sleeve_r_2_capsule
   center=<7.0759 0.2174 0>
   rotation=<rot 0.0329 -0.0092 90>
  }
 ]
 dyn_sleeve_r_3 [{shapeType=Capsule, length=4.6605, radius=2.8011, shapeName=dyn_sleeve_r_3_capsule, center=<3 0 0>, rotation=<rot 0 0 80.5571>}]
 pelvis [{shapeType=Capsule, length=11.404, radius=9.9446, shapeName=pelvis_capsule, center=<-6.565 -1.5191 0.9078>, rotation=<rot 2.1126 -90 -0.0911>}]
 spine_01 [{shapeType=Capsule, length=5.7449, radius=10.3491, shapeName=spine_01_capsule, center=<5.6483 0 0>, rotation=<rot 0 90 90>}]
 thigh_l [{
   shapeType=Capsule
   length=19.4708
   radius=10.6558
   shapeName=thigh_l_capsule
   center=<-16.0386 -3.5234 0.7572>
   rotation=<rot -0.0738 -99.9642 -90.5228>
  }]
 thigh_r [{
   shapeType=Capsule
   length=19.4708
   radius=10.6558
   shapeName=thigh_l_capsule
   center=<16.1014 3.8181 -1.95>
   rotation=<rot -0.0738 -99.9642 89.4772>
  }]
 thigh_twist_01_l [{shapeType=Capsule, length=6.9728, radius=8.204, center=<-10.3096 0 0>, rotation=<rot 90 0 0>}]
 thigh_twist_01_r [{shapeType=Capsule, length=6.9728, radius=8.204, center=<10.3096 0 0>, rotation=<rot 90 0 180>}]
}
constraints [
 {
  bone1Name=base_sleeve_l
  bone2Name=pelvis
  swing1LimitDegrees=45
  swing1Motion=Locked
  swing2LimitDegrees=45
  swing2Motion=Locked
  twistLimitDegrees=45
  twistMotion=Locked
 }
 {bone1Name=dyn_sleeve_l_1, bone2Name=base_sleeve_l, swing1LimitDegrees=12, swing2LimitDegrees=25, twistLimitDegrees=0}
 {bone1Name=dyn_sleeve_l_2, bone2Name=dyn_sleeve_l_1, swing1LimitDegrees=15, swing2LimitDegrees=10, twistLimitDegrees=0}
 {bone1Name=dyn_sleeve_l_3, bone2Name=dyn_sleeve_l_2, swing1LimitDegrees=20, swing2LimitDegrees=20, twistLimitDegrees=0}
 {bone1Name=dyn_sleeve_l_4, bone2Name=dyn_sleeve_l_3, swing1LimitDegrees=20, swing2LimitDegrees=20, twistLimitDegrees=0}
 {
  bone1Name=base_sleeve_r
  bone2Name=pelvis
  swing1LimitDegrees=45
  swing1Motion=Locked
  swing2LimitDegrees=45
  swing2Motion=Locked
  twistLimitDegrees=45
  twistMotion=Locked
 }
 {bone1Name=dyn_sleeve_r_1, bone2Name=base_sleeve_r, swing1LimitDegrees=20, swing2LimitDegrees=25, twistLimitDegrees=0}
 {bone1Name=dyn_sleeve_r_2, bone2Name=dyn_sleeve_r_1, swing1LimitDegrees=20, swing2LimitDegrees=10, twistLimitDegrees=0}
 {bone1Name=dyn_sleeve_r_3, bone2Name=dyn_sleeve_r_2, swing1LimitDegrees=20, swing2LimitDegrees=10, twistLimitDegrees=0}
 {
  bone1Name=thigh_l
  bone2Name=pelvis
  swing1LimitDegrees=45
  swing1Motion=Free
  swing2LimitDegrees=45
  swing2Motion=Free
  twistLimitDegrees=45
  twistMotion=Free
 }
 {
  bone1Name=base_sleeve_l
  bone2Name=pelvis
  swing1LimitDegrees=45
  swing1Motion=Locked
  swing2LimitDegrees=45
  swing2Motion=Locked
  twistLimitDegrees=45
  twistMotion=Locked
 }
 {
  bone1Name=base_sleeve_r
  bone2Name=pelvis
  swing1LimitDegrees=45
  swing1Motion=Locked
  swing2LimitDegrees=45
  swing2Motion=Locked
  twistLimitDegrees=45
  twistMotion=Locked
 }
 {
  bone1Name=thigh_r
  bone2Name=pelvis
  swing1LimitDegrees=45
  swing1Motion=Free
  swing2LimitDegrees=45
  swing2Motion=Free
  twistLimitDegrees=45
  twistMotion=Free
 }
 {
  bone1Name=thigh_twist_01_l
  bone2Name=thigh_l
  swing1LimitDegrees=45
  swing1Motion=Free
  swing2LimitDegrees=45
  swing2Motion=Free
  twistLimitDegrees=45
  twistMotion=Free
 }
 {
  bone1Name=thigh_twist_01_r
  bone2Name=thigh_r
  swing1LimitDegrees=45
  swing1Motion=Free
  swing2LimitDegrees=45
  swing2Motion=Free
  twistLimitDegrees=45
  twistMotion=Free
 }
 {
  bone1Name=spine_01
  bone2Name=pelvis
  swing1LimitDegrees=45
  swing1Motion=Free
  swing2LimitDegrees=45
  swing2Motion=Free
  twistLimitDegrees=45
  twistMotion=Free
 }
]
