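# Physics asset for the back skirt or cloak of the BiscuitFluff female medium character body.
# Capsule bodies form center, left and right back chains, each hanging from a master bone locked to skirt_base, with tight swing and twist limits down the chain.
# Pelvis and thigh capsules provide body volumes around the cloth.
# Skirt_base and the right thigh swing freely from the pelvis, while the left thigh uses narrow limits.
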PhysicsAsset /BRCosmetics/Characters/Player/Female/Medium/Bodies/F_MED_BiscuitFluff/Meshes/F_MED_BiscuitFluff_Physics_cloak
bodies {
 dyn_skirt_bk_01 [{shapeType=Capsule, length=7.9782, radius=6.9249, shapeName=dyn_skirt_bk_01_capsule, center=<4.2906 0 6.784>, rotation=<rot 0 0 90>}]
 dyn_skirt_bk_02 [{shapeType=Capsule, length=7.9782, radius=6.9249, shapeName=dyn_skirt_bk_02_capsule, center=<4.2906 0 6.784>, rotation=<rot 0 0 90>}]
 dyn_skirt_bk_03 [{shapeType=Capsule, length=7.9782, radius=6.9249, shapeName=dyn_skirt_bk_03_capsule, center=<4.2906 0 6.784>, rotation=<rot 0 0 90>}]
 dyn_skirt_bk_l_01 [{shapeType=Capsule, length=7.9782, radius=6.9249, shapeName=dyn_skirt_bk_l_01_capsule, center=<4.2906 0 6.784>, rotation=<rot 0 0 90>}]
 dyn_skirt_bk_l_02 [{shapeType=Capsule, length=5.989, radius=8.9141, shapeName=dyn_skirt_bk_l_02_capsule, center=<5.027 0 8.159>, rotation=<rot 0 0 90>}]
 dyn_skirt_bk_l_mstr [{shapeType=Capsule, length=0.505, radius=0.505, shapeName=dyn_skirt_bk_l_mstr_capsule, rotation=<rot 0 -90 0>}]
 dyn_skirt_bk_mstr [{shapeType=Capsule, length=0.505, radius=0.505, shapeName=dyn_skirt_bk_mstr_capsule, rotation=<rot 0 -90 0>}]
 dyn_skirt_bk_r_01 [{
   shapeType=Capsule
   length=7.9782
   radius=6.9249
   shapeName=dyn_skirt_bk_r_01_capsule
   center=<4.2906 0 7>
   rotation=<rot -0.0031 -0.0016 89.8004>
  }]
 dyn_skirt_bk_r_02 [{shapeType=Capsule, length=6.2439, radius=8.6592, shapeName=dyn_skirt_bk_r_02_capsule, center=<5 0 8>, rotation=<rot 0 0 90>}]
 dyn_skirt_bk_r_mstr [{shapeType=Capsule, length=0.505, radius=0.505, shapeName=dyn_skirt_bk_r_mstr_capsule, rotation=<rot 0 -90 0>}]
 pelvis [{shapeType=Capsule, length=3.349, radius=15.3881, shapeName=pelvis_capsule, center=<-9.9742 2.4143 0.9255>, rotation=<rot -7.2746 -90 1.7711>}]
 skirt_base [{shapeType=Capsule, length=0.505, radius=0.505, shapeName=skirt_base_capsule, rotation=<rot 0 -90 0>}]
 thigh_l [{shapeType=Capsule, length=29.0261, radius=12.1921, center=<-29.6423 0.649 0.0343>, rotation=<rot -89.641 179.9502 0.7422>}]
 thigh_r [{shapeType=Capsule, length=26.5159, radius=11.4025, center=<29.6423 -0.649 -0.0343>, rotation=<rot -89.641 179.9502 -179.2578>}]
}
constraints [
 {
  bone1Name=thigh_r
  bone2Name=pelvis
  swing1LimitDegrees=45
  swing1Motion=Free
  swing2LimitDegrees=45
  swing2Motion=Free
  twistLimitDegrees=45
  twistMotion=Free
 }
 {bone1Name=dyn_skirt_bk_01, bone2Name=dyn_skirt_bk_mstr, swing1LimitDegrees=5, swing2LimitDegrees=15, twistLimitDegrees=10}
 {bone1Name=dyn_skirt_bk_02, bone2Name=dyn_skirt_bk_01, swing1LimitDegrees=10, swing2LimitDegrees=15, twistLimitDegrees=1}
 {bone1Name=dyn_skirt_bk_03, bone2Name=dyn_skirt_bk_02, swing1LimitDegrees=5, swing2LimitDegrees=15, twistLimitDegrees=1}
 {
  bone1Name=dyn_skirt_bk_mstr
  bone2Name=skirt_base
  swing1LimitDegrees=45
  swing1Motion=Locked
  swing2LimitDegrees=45
  swing2Motion=Locked
  twistLimitDegrees=45
  twistMotion=Locked
 }
 {bone1Name=dyn_skirt_bk_r_01, bone2Name=dyn_skirt_bk_r_mstr, swing1LimitDegrees=2, swing2LimitDegrees=15, twistLimitDegrees=1}
 {bone1Name=dyn_skirt_bk_r_02, bone2Name=dyn_skirt_bk_r_01, swing1LimitDegrees=2, swing2LimitDegrees=45, twistLimitDegrees=1}
 {
  bone1Name=dyn_skirt_bk_r_mstr
  bone2Name=skirt_base
  swing1LimitDegrees=45
  swing1Motion=Locked
  swing2LimitDegrees=45
  swing2Motion=Locked
  twistLimitDegrees=45
  twistMotion=Locked
 }
 {bone1Name=dyn_skirt_bk_l_01, bone2Name=dyn_skirt_bk_l_mstr, swing1LimitDegrees=3, swing2LimitDegrees=15, twistLimitDegrees=1}
 {bone1Name=dyn_skirt_bk_l_02, bone2Name=dyn_skirt_bk_l_01, swing1LimitDegrees=3, swing2LimitDegrees=45, twistLimitDegrees=1}
 {
  bone1Name=dyn_skirt_bk_l_mstr
  bone2Name=skirt_base
  swing1LimitDegrees=45
  swing1Motion=Locked
  swing2LimitDegrees=45
  swing2Motion=Locked
  twistLimitDegrees=45
  twistMotion=Locked
 }
 {
  bone1Name=skirt_base
  bone2Name=pelvis
  swing1LimitDegrees=45
  swing1Motion=Free
  swing2LimitDegrees=45
  swing2Motion=Free
  twistLimitDegrees=45
  twistMotion=Free
 }
 {bone1Name=thigh_l, bone2Name=pelvis, swing1LimitDegrees=5, swing2LimitDegrees=15, twistLimitDegrees=1}
]
