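# Animation Blueprint for the F_MED_BiscuitFluff female player body, adding procedural and physics motion, including a cloak, over the attached parent mesh's pose.
# The copied pose runs through a Control Rig fed movement, character state, LOD and alt-skin flags, then a RigidBody pass, and is cached.
# Two cloak RigidBody nodes simulate from that cache in pelvis space, blended by DBNO state.
# The computed gravity override is stored but not wired into the RigidBody nodes.
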AnimBlueprint /BRCosmetics/Characters/Player/Female/Medium/Bodies/F_MED_BiscuitFluff/Meshes/F_MED_BiscuitFluff_AnimBP
target_skeleton=&/F_MED_BiscuitFluff_Skeleton
graphs [
 AnimGraph {nodes [
   n1 = RigidBody!(ComponentPose=Local To Component(LocalPose=Use cached pose 'base'()), Alpha, OverrideWorldGravity:Vector="0,0,0", ExternalForce) -> (Pose) {
    bClampLinearTranslationLimitToRefPose=true
    bOverrideWorldGravity=true
    overridePhysicsAsset=&/F_MED_BiscuitFluff_Physics_cloak
    simulationSpace=BaseBoneSpace
    alphaBoolBlend {blendOption=Linear}
    baseBoneRef {boneName=pelvis}
    componentLinearAccScale=<0.1 0.1 0.1>
    componentLinearVelScale=<0.1 0.1 0.1>
    simSpaceSettings {
     maxAngularAcceleration=500
     maxAngularVelocity=3
     maxLinearAcceleration=400
     maxLinearVelocity=250
     velocityScaleZ=0
     worldAlpha=1
     externalLinearDragV=<0.1 0.1>
    }
   }
   n2 = RigidBody(ComponentPose=Local To Component(LocalPose=Use cached pose 'base'()), OverrideWorldGravity="0,0,0") {
    bClampLinearTranslationLimitToRefPose=true
    bOverrideWorldGravity=true
    overridePhysicsAsset=&/F_MED_BiscuitFluff_Physics_cloak_DBNO
    simulationSpace=BaseBoneSpace
    alphaBoolBlend {blendOption=Linear}
    baseBoneRef {boneName=pelvis}
    componentLinearAccScale=<0.1 0.1 0.1>
    componentLinearVelScale=<0.1 0.1 0.1>
    simSpaceSettings {
     maxAngularAcceleration=500
     maxAngularVelocity=3
     maxLinearAcceleration=400
     maxLinearVelocity=250
     velocityScaleZ=0
     worldAlpha=1
     externalLinearDragV=<0.1 0.1>
    }
   }
   Output Pose(Result=Blend Poses by bool(BlendPose_0=Component To Local(ComponentPose=@n2), BlendPose_1=Component To Local(ComponentPose=@n1), bActiveValue=Get bIsDBNO()) {blendType=Linear}) {layerGroup=None}
   n3 = Control Rig!(
    Source=Copy Pose From Mesh() {bUseAttachedParent=true}
    Alpha
    Forward Backward Movement:double=Get Forward Backward Movement() -> (Forward Backward Movement:double)
    Left Right Movement:double=Get Left Right Movement() -> (Left Right Movement:double)
    isEmote:bool=Get bIsPlayingEmote()
    isSkyDive:bool=Get bIsSkyDiving()
    isAltskin:bool=Get isALTsskin() -> (isALTsskin:bool)
    isLOD2:bool=Get isLOD2() -> (isLOD2:bool)
    isFloating:bool=Get bIsSkydiveFloating()
    isDBNO:bool=Get bIsDBNO()
   ) -> (Pose) {
    alphaBoolBlend {blendOption=Linear}
    controlRigAssetReference {blueprintRigClass=&/F_MED_BiscuitFluff_CtrlRig.F_MED_BiscuitFluff_CtrlRig_C}
   }
   n4 = RigidBody(ComponentPose=Local To Component(LocalPose=@n3), OverrideWorldGravity="0,0,0") {
    bClampLinearTranslationLimitToRefPose=true
    bOverrideWorldGravity=true
    overridePhysicsAsset=&/F_MED_BiscuitFluff_Physics
    simulationSpace=BaseBoneSpace
    alphaBoolBlend {blendOption=Linear}
    baseBoneRef {boneName=pelvis}
   }
   Save cached pose 'base'(Pose=Component To Local(ComponentPose=@n4)) {cacheName=base}
  ]}
 EventGraph {nodes [
   Event BlueprintUpdateAnimation()
   > n1 = Gravity Override State Function DEPRECATED(CharacterPartAnimInstance=Self-Reference(), GameTimeInSeconds=GetGameTimeInSeconds(), GravityOverrideParams=Get Gravity Override Params())
   > Set Gravity Override(Gravity Override=@n1)
   > n2 = PlayerMovement(AnimInstance=Self-Reference(), XDotProductCoefficient=1, YDotProductCoefficient=1, ZDotProductCoefficient=1)
   > Set Left Right Movement(Left Right Movement:double=@n2) -> (Output_Get:double)
   > Set Forward Backward Movement(Forward Backward Movement:double=@n2.YMovement) -> (Output_Get:double)
   > Set isALTsskin(isALTsskin:bool=NOT Boolean(A=Get bRigidBodyEnabled())) -> (Output_Get:bool)
   > Set isLOD2(isLOD2:bool="integer >= integer"(A=Get CurrentLOD(), B=2)) -> (Output_Get:bool)
  ]}
]
variables [
 Forward Backward Movement {type=number}
 Gravity Override {type=Vector}
 Gravity Override Params {
  type=GravityOverrideParamsStruct
  default {
   skydive_FloatingDown {gravityOverride=<0 0 0>, windOffset=<0 -1200 -100>}
   skydive_GoingUp {gravityOverride=<0 -300 100>}
   skydive_Parachuting {gravityOverride=<0 -200 -2000>}
  }
 }
 isALTsskin {type=boolean}
 isLOD2 {type=boolean}
 Left Right Movement {type=number}
]
