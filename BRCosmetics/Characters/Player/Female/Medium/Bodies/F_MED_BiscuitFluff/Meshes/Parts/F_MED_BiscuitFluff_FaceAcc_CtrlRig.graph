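# Control Rig for the face accessory part of the BiscuitFluff female character, adjusting accessory bones such as a hair bundle and bow as the head moves.
# A spherical pose reader on the head bone outputs a weight from an offset active region with falloff.
# That weight blends preset transform offsets onto dyn_bundle1, Master_Bow_01_r and other listed items through Modify Transforms.
# A second, identity-valued Modify Transforms node is left unconnected.
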ControlRigBlueprint /BRCosmetics/Characters/Player/Female/Medium/Bodies/F_MED_BiscuitFluff/Meshes/Parts/F_MED_BiscuitFluff_FaceAcc_CtrlRig
graphs [
 RigVMFunctionLibrary {links [], nodes []}
 RigVMModel {nodes [
   RigUnit_BeginExecution = &RigUnit_BeginExecution()
   n1 = &RigUnit_SphericalPoseReader(
    ActiveRegionScaleFactors=(PositiveWidth=1,NegativeWidth=1,PositiveHeight=1,NegativeHeight=1.8)
    ActiveRegionSize=0.1
    Debug=(bDrawDebug=true,bDraw2D=False,bDrawLocalAxes=False,DebugScale=25,DebugSegments=20,DebugThickness=0.25)
    DriverAxis=(X=1,Y=0,Z=0)
    DriverItem=(Type=Bone,Name="head")
    FalloffRegionScaleFactors=(PositiveWidth=1,NegativeWidth=1,PositiveHeight=1,NegativeHeight=1)
    FalloffSize=0.25
    FlipHeightScaling=False
    FlipWidthScaling=False
    OptionalParentItem=(Type=Bone,Name="None")
    RotationOffset=(X=90,Y=-20,Z=-115)
    ExecutePin=@RigUnit_BeginExecution.ExecutePin
   )
   > &RigUnit_ModifyTransforms(
    ItemToModify="((Item=(Type=Bone,Name=\"dyn_bundle1\"),Transform=(Rotation=(X=0,Y=0.4617,Z=0,W=0.887),Translation=(X=-0.379,Y=-0.8222,Z=1.4036),Scale3D=(X=1,Y=1,Z=1))),(Item=(Type=Bone,Name=\"Master_Bow_01_r\"),Transform=(Rotation=(X=0,Y=0,Z=-0.4196,W=0.9077),Translation=(X= ...(truncated)"
    WeightMinimum=0
    Weight=@n1.OutputParam
   )
   &RigUnit_ModifyTransforms(
    ExecutePin=()
    ItemToModify=((Item=(Type=Bone,Name="dyn_bundle1"),Transform=(Rotation=(X=0,Y=0,Z=0,W=1),Translation=(X=0,Y=0,Z=0),Scale3D=(X=1,Y=1,Z=1))))
    Weight=1
    WeightMinimum=0
   )
  ]}
]
hierarchy [
 root {loc=<0 0 0>, rot=<0 0 0 1>}
 pelvis {parent=root, loc=<0 -0.5301 98.0571>, rot=<0.0127 0.707 -0.0127 -0.707>}
 spine_01 {parent=pelvis, loc=<6.1208 -0.9328 0>, rot=<0 0 0.034 -0.9994>}
 spine_02 {parent=spine_01, loc=<6.5997 -0.2095 0>, rot=<0 0 0.0238 -0.9997>}
 spine_03 {parent=spine_02, loc=<6.4914 -0.1237 0>, rot=<0 0 0.0149 -0.9999>}
 spine_04 {parent=spine_03, loc=<6.9598 0.0794 0>, rot=<0 0 -0.0132 -0.9999>}
 spine_05 {parent=spine_04, loc=<6.9987 0 0>, rot=<0 0 -0.0015 -1>}
 attach_cape {parent=spine_05, loc=<9.3216 -8.4501 0>, rot=<0.0616 -0.7044 -0.0616 -0.7044>}
 neck_01 {parent=spine_05, loc=<11.3138 0.1529 0>, rot=<0 0 -0.2003 -0.9797>}
 neck_02 {parent=neck_01, loc=<4.5627 0 0>, rot=<0 0 0.0504 -0.9987>}
 head {parent=neck_02, loc=<5.3914 0 0>, rot=<0 0 0.115 -0.9934>}
 Hair_04_r_big {parent=head, loc=<4.5984 3.9788 6.6221>, rot=<0.0047 -0.9993 -0.0329 -0.0184>}
 Hair_03_r_big {parent=Hair_04_r_big, loc=<3.948 0 0>, rot=<0.0278 0.0163 0.0005 -0.9995>}
 Hair_02_r_big {parent=Hair_03_r_big, loc=<3.948 0 0>, rot=<0.0028 0.0533 0.0533 -0.9972>}
 Hair_01_r_big {parent=Hair_02_r_big, loc=<3.948 0 0>, rot=<0 0 0 1>}
 Hair_01_r_fr {parent=head, loc=<5.6039 7.968 5.3674>, rot=<0.0108 -0.9817 -0.0338 0.1869>}
 Hair_02_r_fr {parent=Hair_01_r_fr, loc=<4.3642 0 0.0135>, rot=<0.0278 0.0163 0.0005 -0.9995>}
 Hair_01_l_big {parent=head, loc=<2.5865 2.5798 -7.3919>, rot=<0.0403 -0.0432 0.9982 -0.0109>}
 Hair_02_l_big {parent=Hair_01_l_big, loc=<5.4922 0 0>, rot=<0.0002 0.0337 0.0063 -0.9994>}
 Hair_03_l_big {parent=Hair_02_l_big, loc=<4.9043 -0.1642 0>, rot=<0 0 0.0849 -0.9964>}
 Hair_01_mid {parent=head, loc=<7.9564 8.9245 -0.7074>, rot=<-0.1214 -0.6966 0.7056 0.046>}
 Hair_02_mid {parent=Hair_01_mid, loc=<2.8209 0 0>, rot=<0 0.0788 0 -0.9969>}
 Hair_03_mid {parent=head, loc=<7.7365 9.0344 1.9878>, rot=<-0.1548 -0.789 0.5893 0.079>}
 Hair_01_l_fr {parent=head, loc=<4.8141 5.1208 -6.6056>, rot=<-0.019 -0.1848 0.9786 0.0882>}
 Hair_02_l_fr {parent=Hair_01_l_fr, loc=<4.2949 0 0>, rot=<0 0 0 1>}
 dyn_bundle {parent=head, loc=<11.5118 -6.2449 -0.2962>, rot=<-0.6009 0.2307 0.2946 -0.7063>}
 dyn_bundle1 {parent=head, loc=<1.1836 -9.081 1.4609>, rot=<-0.6693 -0.0224 0.0136 -0.7425>}
 Master_Bow_01_r {parent=dyn_bundle1, loc=<-1.5986 0.6777 -0.1074>, rot=<-0.334 -0.9293 0.1571 -0.0109>}
 Bow_01_r {parent=Master_Bow_01_r, loc=<0 0 0>, rot=<0 0 0 1>}
 Bow_02_r {parent=Bow_01_r, loc=<3.8534 0 -0.0529>, rot=<0.0002 -0.0244 -0.0222 -0.9995>}
 Master_Bow_01_l {parent=dyn_bundle1, loc=<-0.7527 -2.6484 0.4587>, rot=<0.4669 -0.8809 -0.0769 -0.0127>}
 Bow_01_l {parent=Master_Bow_01_l, loc=<0 0 0>, rot=<0 0 0 1>}
 Bow_02_l {parent=Bow_01_l, loc=<4.7649 0 -0.3184>, rot=<0 -0.1412 0 -0.99>}
 dyn_lok_01 {parent=head, loc=<10.5578 -7.2979 -3.6502>, rot=<-0.5257 0.4837 0.6995 -0.0177>}
 clavicle_l {parent=spine_05, loc=<5.4831 -0.1053 -3.2946>, rot=<0.0465 -0.7289 -0.009 -0.6829>}
 clavicle_r {parent=spine_05, loc=<5.4831 -0.1053 3.2946>, rot=<0.7289 0.0465 -0.6829 0.009>}
 upperarm_r {parent=clavicle_r, loc=<-9.8658 1.2136 0.4105>, rot=<0.0164 -0.3245 -0.0269 -0.9454>}
 upperarm_twist_01_r {parent=upperarm_r, loc=<-7.2135 0.1134 0>, rot=<0 0 0 -1>}
 deform_upperarm_up_base_r {parent=upperarm_twist_01_r, loc=<8.891 -0.4825 -3.9517>, rot=<0 0 0 1>}
 deform_upperarm_up_r {parent=deform_upperarm_up_base_r, loc=<0 0 0>, rot=<0 0 0 1>}
]
