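# Physics asset for the face accessory mesh of the female BiscuitFluff character body, setting up physics for its left and right bows.
# Each bow is a two-bone capsule chain constrained to the head with narrow swing and twist limits, slightly looser on the right.
# Capsules on the head, upper spine, both clavicles and right upper arm form the body around the bows.
# The head rotates freely against the spine, the clavicles are rotation-locked, and the arm can only twist.
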PhysicsAsset /BRCosmetics/Characters/Player/Female/Medium/Bodies/F_MED_BiscuitFluff/Meshes/Parts/F_MED_BiscuitFluff_FaceAcc_Physics
bodies {
 Bow_01_l [{
   shapeType=Capsule
   length=2.7455
   radius=2.6155
   shapeName=Bow_01_l_capsule
   center=<2.611 -0.4701 -0.1746>
   rotation=<rot 6.8676 111.5068 -90.9102>
  }]
 Bow_01_r [{
   shapeType=Capsule
   length=2.4384
   radius=1.8213
   shapeName=Bow_01_r_capsule
   center=<1.9996 -0.0365 0.1212>
   rotation=<rot 7.5591 -93.441 -85.8659>
  }]
 Bow_02_l [{
   shapeType=Capsule
   length=3.4251
   radius=2.6876
   shapeName=Bow_02_l_capsule
   center=<3.3028 0.4615 -0.62>
   rotation=<rot 0 -81.2891 -73.8615>
  }]
 Bow_02_r [{
   shapeType=Capsule
   length=3.1678
   radius=1.8916
   shapeName=Bow_02_r_capsule
   center=<3.115 -0.1928 0.0759>
   rotation=<rot 0 -87.5433 -85.3471>
  }]
 clavicle_l [{shapeType=Capsule, length=3.4779, radius=5.9881, shapeName=clavicle_r_capsule, center=<4.7769 0 1.1888>, rotation=<rot -0.0001 -90 94.2256>}]
 clavicle_r [{
   shapeType=Capsule
   length=3.7092
   radius=5.7568
   shapeName=clavicle_r_capsule
   center=<-4.7769 0 -1.1888>
   rotation=<rot -0.0001 -90 -85.7744>
  }]
 head [{shapeType=Capsule, length=0.732, radius=8.1948, center=<2.1846 0.1774 0.1725>, rotation=<rot 7.8321 95.8206 -90>}]
 spine_05 [{shapeType=Capsule, length=15.326, radius=11.8696, shapeName=spine_05_capsule, center=<3.2046 5.2579 0>, rotation=<rot 0 -90 0>}]
 upperarm_r [{shapeType=Capsule, length=6.6116, radius=4.7568, shapeName=upperarm_r_capsule, center=<-6.4771 0 -0.2514>, rotation=<rot 0 -90 87.777>}]
}
constraints [
 {
  bone1Name=head
  bone2Name=spine_05
  swing1LimitDegrees=45
  swing1Motion=Free
  swing2LimitDegrees=45
  swing2Motion=Free
  twistLimitDegrees=45
  twistMotion=Free
 }
 {bone1Name=Bow_01_l, bone2Name=head, swing1LimitDegrees=10, swing2LimitDegrees=15, twistLimitDegrees=5}
 {bone1Name=Bow_02_l, bone2Name=Bow_01_l, swing1LimitDegrees=10, swing2LimitDegrees=15, twistLimitDegrees=5}
 {bone1Name=Bow_01_r, bone2Name=head, swing1LimitDegrees=15, swing2LimitDegrees=30, twistLimitDegrees=5}
 {bone1Name=Bow_02_r, bone2Name=Bow_01_r, swing1LimitDegrees=10, swing2LimitDegrees=20, twistLimitDegrees=5}
 {
  bone1Name=clavicle_r
  bone2Name=spine_05
  swing1LimitDegrees=45
  swing1Motion=Locked
  swing2LimitDegrees=45
  swing2Motion=Locked
  twistLimitDegrees=45
  twistMotion=Locked
 }
 {
  bone1Name=clavicle_l
  bone2Name=spine_05
  swing1LimitDegrees=45
  swing1Motion=Locked
  swing2LimitDegrees=45
  swing2Motion=Locked
  twistLimitDegrees=45
  twistMotion=Locked
 }
 {
  bone1Name=upperarm_r
  bone2Name=clavicle_r
  swing1LimitDegrees=45
  swing1Motion=Locked
  swing2LimitDegrees=45
  swing2Motion=Locked
  twistLimitDegrees=45
  twistMotion=Free
 }
]
