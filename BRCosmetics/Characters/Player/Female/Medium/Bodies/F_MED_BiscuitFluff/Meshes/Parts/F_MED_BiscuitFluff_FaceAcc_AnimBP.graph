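# Post-process style AnimBP for the BiscuitFluff female character's face accessory mesh, adding secondary motion to hair and dangling bones.
# It copies the attached parent's pose, runs a Control Rig, then chained per-bone AnimDynamics and a partial-alpha RigidBody sim.
# The Bow AnimDynamics chain is not connected to the output pose.
# The event graph updates a Gravity Override variable each frame through a deprecated state function.
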AnimBlueprint /BRCosmetics/Characters/Player/Female/Medium/Bodies/F_MED_BiscuitFluff/Meshes/Parts/F_MED_BiscuitFluff_FaceAcc_AnimBP
target_skeleton=&/F_MED_BiscuitFluff_FaceAcc_Skeleton
graphs [
 AnimGraph {nodes [
   n1 = Anim Dynamics - Bone: Bow_01_l!(
    ComponentPose=Reroute Node(InputPin:Component Space Pose Link Structure) -> (OutputPin:Component Space Pose Link Structure)
    bAlphaBoolEnabled:bool=True
    ExternalForce="0,0,0"
   ) -> (Pose) {
    alphaInputType=Bool
    angularDampingOverride=0.9999
    bOverrideAngularDamping=true
    bUseGravityOverride=true
    simulationSpace=RootRelative
    alphaBoolBlend {blendInTime=0.1, blendOption=Linear, blendOutTime=0.2}
    boundBone {boneName=Bow_01_l}
    componentAppliedLinearAccClamp=<1500 1500 1500>
    componentLinearAccScale=<0.2 0.2 0.2>
    componentLinearVelScale=<0.3 0.3 0.3>
    gravityOverride {z=-980}
    physicsBodyDefinitions [{
      boundBone {boneName=Bow_01_l}
      boxExtents=<12 7 7>
      constraintSetup {angularLimitsMax=<5 10 3>, angularLimitsMin=<-5 -10 -3>, angularTarget=<0 0 0>}
      localJointOffset=<1 0 0>
     }]
   }
   n2 = Anim Dynamics - Bone: Bow_02_l!(ComponentPose=@n1, bAlphaBoolEnabled:bool=True, ExternalForce="0,0,0") -> (Pose) {
    alphaInputType=Bool
    angularDampingOverride=0.9999
    bOverrideAngularDamping=true
    bUseGravityOverride=true
    simulationSpace=RootRelative
    alphaBoolBlend {blendInTime=0.1, blendOption=Linear, blendOutTime=0.2}
    boundBone {boneName=Bow_02_l}
    componentAppliedLinearAccClamp=<1500 1500 1500>
    componentLinearAccScale=<0.2 0.2 0.2>
    componentLinearVelScale=<0.3 0.3 0.3>
    gravityOverride {z=-980}
    physicsBodyDefinitions [{
      boundBone {boneName=Bow_02_l}
      boxExtents=<12 7 7>
      constraintSetup {angularLimitsMax=<5 10 3>, angularLimitsMin=<-5 -12 -3>, angularTarget=<0 0 0>}
      localJointOffset=<1 0 0>
     }]
   }
   n3 = Anim Dynamics - Bone: Bow_01_r!(ComponentPose=@n2, bAlphaBoolEnabled:bool=True, ExternalForce="0,0,0") -> (Pose) {
    alphaInputType=Bool
    angularDampingOverride=0.9999
    bOverrideAngularDamping=true
    bUseGravityOverride=true
    simulationSpace=RootRelative
    alphaBoolBlend {blendInTime=0.1, blendOption=Linear, blendOutTime=0.2}
    boundBone {boneName=Bow_01_r}
    componentAppliedLinearAccClamp=<1500 1500 1500>
    componentLinearAccScale=<0.2 0.2 0.2>
    componentLinearVelScale=<0.3 0.3 0.3>
    gravityOverride {z=-980}
    physicsBodyDefinitions [{
      boundBone {boneName=Bow_01_r}
      boxExtents=<12 7 7>
      constraintSetup {angularLimitsMax=<5 12 10>, angularLimitsMin=<-5 -1 -1>, angularTarget=<0 0 0>}
      localJointOffset=<1 0 0>
     }]
   }
   Anim Dynamics - Bone: Bow_02_r!(ComponentPose=@n3, bAlphaBoolEnabled:bool=True, ExternalForce="0,0,0") -> (Pose) {
    alphaInputType=Bool
    angularDampingOverride=0.9999
    bOverrideAngularDamping=true
    bUseGravityOverride=true
    simulationSpace=RootRelative
    alphaBoolBlend {blendInTime=0.1, blendOption=Linear, blendOutTime=0.2}
    boundBone {boneName=Bow_02_r}
    componentAppliedLinearAccClamp=<1500 1500 1500>
    componentLinearAccScale=<0.2 0.2 0.2>
    componentLinearVelScale=<0.3 0.3 0.3>
    gravityOverride {z=-980}
    physicsBodyDefinitions [{
      boundBone {boneName=Bow_02_r}
      boxExtents=<12 7 7>
      constraintSetup {angularLimitsMax=<5 12 3>, angularLimitsMin=<-5 -1 -3>, angularTarget=<0 0 0>}
      localJointOffset=<1 0 0>
     }]
   }
   n4 = Control Rig(Source=Copy Pose From Mesh() {bUseAttachedParent=true}) {
    alphaBoolBlend {blendOption=Linear}
    controlRigAssetReference {blueprintRigClass=&/F_MED_BiscuitFluff_FaceAcc_CtrlRig.F_MED_BiscuitFluff_FaceAcc_CtrlRig_C}
   }
   n5 = Anim Dynamics - Bone: Hair_04_r_big!(ComponentPose=Local To Component(LocalPose=@n4), bAlphaBoolEnabled:bool=True, ExternalForce="0,0,0") -> (Pose) {
    alphaInputType=Bool
    angularDampingOverride=0.9999
    bOverrideAngularDamping=true
    bUseGravityOverride=true
    simulationSpace=RootRelative
    alphaBoolBlend {blendInTime=0.1, blendOption=Linear, blendOutTime=0.2}
    boundBone {boneName=Hair_04_r_big}
    componentAppliedLinearAccClamp=<1500 1500 1500>
    componentLinearAccScale=<0.2 0.2 0.2>
    componentLinearVelScale=<0.3 0.3 0.3>
    gravityOverride {z=-980}
    physicsBodyDefinitions [{
      boundBone {boneName=Hair_04_r_big}
      boxExtents=<12 7 7>
      constraintSetup {angularLimitsMax=<1 12 30>, angularLimitsMin=<-1 -2 -11>, angularTarget=<0 0 0>}
      localJointOffset=<1 0 0>
     }]
   }
   n6 = Anim Dynamics - Bone: Hair_03_r_big!(ComponentPose=@n5, bAlphaBoolEnabled:bool=True, ExternalForce="0,0,0") -> (Pose) {
    alphaInputType=Bool
    angularDampingOverride=0.9999
    bOverrideAngularDamping=true
    bUseGravityOverride=true
    simulationSpace=RootRelative
    alphaBoolBlend {blendInTime=0.5, blendOption=Linear, blendOutTime=0.6}
    boundBone {boneName=Hair_03_r_big}
    componentAppliedLinearAccClamp=<500 500 500>
    componentLinearAccScale=<0.2 0.2 0.2>
    componentLinearVelScale=<0.3 0.3 0.3>
    gravityOverride {z=-980}
    physicsBodyDefinitions [{
      boundBone {boneName=Hair_03_r_big}
      boxExtents=<15 6 10>
      constraintSetup {angularLimitsMax=<5 15 15>, angularLimitsMin=<-5 -15 -3>, angularTarget=<0 0 0>}
      localJointOffset=<1 0 0>
     }]
   }
   n7 = Anim Dynamics - Bone: Hair_02_r_big!(ComponentPose=@n6, bAlphaBoolEnabled:bool=True, ExternalForce="0,0,0") -> (Pose) {
    alphaInputType=Bool
    angularDampingOverride=0.9999
    bOverrideAngularDamping=true
    bUseGravityOverride=true
    simulationSpace=RootRelative
    alphaBoolBlend {blendInTime=0.4, blendOption=Linear, blendOutTime=0.5}
    boundBone {boneName=Hair_02_r_big}
    componentAppliedLinearAccClamp=<500 500 500>
    componentLinearAccScale=<0.2 0.2 0.2>
    componentLinearVelScale=<0.3 0.3 0.3>
    gravityOverride {z=-980}
    physicsBodyDefinitions [{
      boundBone {boneName=Hair_02_r_big}
      boxExtents=<14 5 5>
      constraintSetup {angularLimitsMax=<5 0 12>, angularLimitsMin=<-5 -15 -3>, angularTarget=<0 0 0>}
      localJointOffset=<1 0 0>
     }]
   }
   n8 = Anim Dynamics - Bone: Hair_01_r_big!(ComponentPose=@n7, bAlphaBoolEnabled:bool=True, ExternalForce="0,0,0") -> (Pose) {
    alphaInputType=Bool
    angularDampingOverride=0.9999
    bOverrideAngularDamping=true
    bUseGravityOverride=true
    simulationSpace=RootRelative
    alphaBoolBlend {blendInTime=0.4, blendOption=Linear, blendOutTime=0.5}
    boundBone {boneName=Hair_01_r_big}
    componentAppliedLinearAccClamp=<500 500 500>
    componentLinearAccScale=<0.2 0.2 0.2>
    componentLinearVelScale=<0.3 0.3 0.3>
    gravityOverride {z=-980}
    physicsBodyDefinitions [{
      boundBone {boneName=Hair_01_r_big}
      boxExtents=<14 5 5>
      constraintSetup {angularLimitsMax=<5 0 12>, angularLimitsMin=<-5 -15 -3>, angularTarget=<0 0 0>}
      localJointOffset=<1 0 0>
     }]
   }
   n9 = Anim Dynamics - Bone: Hair_01_r_fr!(ComponentPose=@n8, bAlphaBoolEnabled:bool=True, ExternalForce="0,0,0") -> (Pose) {
    alphaInputType=Bool
    angularDampingOverride=0.9999
    bOverrideAngularDamping=true
    bUseGravityOverride=true
    simulationSpace=RootRelative
    alphaBoolBlend {blendInTime=0.1, blendOption=Linear, blendOutTime=0.2}
    boundBone {boneName=Hair_01_r_fr}
    componentAppliedLinearAccClamp=<1500 1500 1500>
    componentLinearAccScale=<0.2 0.2 0.2>
    componentLinearVelScale=<0.3 0.3 0.3>
    gravityOverride {z=-980}
    physicsBodyDefinitions [{
      boundBone {boneName=Hair_01_r_fr}
      boxExtents=<12 7 7>
      constraintSetup {angularLimitsMax=<5 10 30>, angularLimitsMin=<-5 -12 -3>, angularTarget=<0 0 0>}
      localJointOffset=<1 0 0>
     }]
   }
   n10 = Anim Dynamics - Bone: Hair_02_r_fr!(ComponentPose=@n9, bAlphaBoolEnabled:bool=True, ExternalForce="0,0,0") -> (Pose) {
    alphaInputType=Bool
    angularDampingOverride=0.9999
    bOverrideAngularDamping=true
    bUseGravityOverride=true
    simulationSpace=RootRelative
    alphaBoolBlend {blendInTime=0.5, blendOption=Linear, blendOutTime=0.6}
    boundBone {boneName=Hair_02_r_fr}
    componentAppliedLinearAccClamp=<500 500 500>
    componentLinearAccScale=<0.2 0.2 0.2>
    componentLinearVelScale=<0.3 0.3 0.3>
    gravityOverride {z=-980}
    physicsBodyDefinitions [{
      boundBone {boneName=Hair_02_r_fr}
      boxExtents=<15 6 10>
      constraintSetup {angularLimitsMax=<5 15 15>, angularLimitsMin=<-5 -15 -3>, angularTarget=<0 0 0>}
      localJointOffset=<1 0 0>
     }]
   }
   n11 = Anim Dynamics - Bone: Hair_01_l_fr!(ComponentPose=@n10, bAlphaBoolEnabled:bool=True, ExternalForce="0,0,0") -> (Pose) {
    alphaInputType=Bool
    angularDampingOverride=0.9999
    bOverrideAngularDamping=true
    bUseGravityOverride=true
    simulationSpace=RootRelative
    alphaBoolBlend {blendInTime=0.1, blendOption=Linear, blendOutTime=0.2}
    boundBone {boneName=Hair_01_l_fr}
    componentAppliedLinearAccClamp=<1500 1500 1500>
    componentLinearAccScale=<0.2 0.2 0.2>
    componentLinearVelScale=<0.3 0.3 0.3>
    gravityOverride {z=-980}
    physicsBodyDefinitions [{
      boundBone {boneName=Hair_01_l_fr}
      boxExtents=<12 7 7>
      constraintSetup {angularLimitsMax=<5 10 10>, angularLimitsMin=<-5 -5 -10>, angularTarget=<0 0 0>}
      localJointOffset=<1 0 0>
     }]
   }
   n12 = Anim Dynamics - Bone: Hair_02_l_fr!(ComponentPose=@n11, bAlphaBoolEnabled:bool=True, ExternalForce="0,0,0") -> (Pose) {
    alphaInputType=Bool
    angularDampingOverride=0.9999
    bOverrideAngularDamping=true
    bUseGravityOverride=true
    simulationSpace=RootRelative
    alphaBoolBlend {blendInTime=0.5, blendOption=Linear, blendOutTime=0.6}
    boundBone {boneName=Hair_02_l_fr}
    componentAppliedLinearAccClamp=<500 500 500>
    componentLinearAccScale=<0.2 0.2 0.2>
    componentLinearVelScale=<0.3 0.3 0.3>
    gravityOverride {z=-980}
    physicsBodyDefinitions [{
      boundBone {boneName=Hair_02_l_fr}
      boxExtents=<15 6 10>
      constraintSetup {angularLimitsMax=<1 15 3>, angularLimitsMin=<-1 -5 -15>, angularTarget=<0 0 0>}
      localJointOffset=<1 0 0>
     }]
   }
   n13 = Anim Dynamics - Bone: Hair_01_l_big!(ComponentPose=@n12, bAlphaBoolEnabled:bool=True, ExternalForce="0,0,0") -> (Pose) {
    alphaInputType=Bool
    angularDampingOverride=0.9999
    bOverrideAngularDamping=true
    bUseGravityOverride=true
    simulationSpace=RootRelative
    alphaBoolBlend {blendInTime=0.1, blendOption=Linear, blendOutTime=0.2}
    boundBone {boneName=Hair_01_l_big}
    componentAppliedLinearAccClamp=<1500 1500 1500>
    componentLinearAccScale=<0.2 0.2 0.2>
    componentLinearVelScale=<0.3 0.3 0.3>
    gravityOverride {z=-980}
    physicsBodyDefinitions [{
      boundBone {boneName=Hair_01_l_big}
      boxExtents=<12 7 7>
      constraintSetup {angularLimitsMax=<5 10 3>, angularLimitsMin=<-5 -1 -30>, angularTarget=<0 0 0>}
      localJointOffset=<1 0 0>
     }]
   }
   n14 = Anim Dynamics - Bone: Hair_02_l_big!(ComponentPose=@n13, bAlphaBoolEnabled:bool=True, ExternalForce="0,0,0") -> (Pose) {
    alphaInputType=Bool
    angularDampingOverride=0.9999
    bOverrideAngularDamping=true
    bUseGravityOverride=true
    simulationSpace=RootRelative
    alphaBoolBlend {blendInTime=0.5, blendOption=Linear, blendOutTime=0.6}
    boundBone {boneName=Hair_02_l_big}
    componentAppliedLinearAccClamp=<500 500 500>
    componentLinearAccScale=<0.2 0.2 0.2>
    componentLinearVelScale=<0.3 0.3 0.3>
    gravityOverride {z=-980}
    physicsBodyDefinitions [{
      boundBone {boneName=Hair_02_l_big}
      boxExtents=<15 6 10>
      constraintSetup {angularLimitsMax=<5 15 3>, angularLimitsMin=<-5 -2 -15>, angularTarget=<0 0 0>}
      localJointOffset=<1 0 0>
     }]
   }
   n15 = Anim Dynamics - Bone: Hair_03_l_big!(ComponentPose=@n14, bAlphaBoolEnabled:bool=True, ExternalForce="0,0,0") -> (Pose) {
    alphaInputType=Bool
    angularDampingOverride=0.9999
    bOverrideAngularDamping=true
    bUseGravityOverride=true
    simulationSpace=RootRelative
    alphaBoolBlend {blendInTime=0.4, blendOption=Linear, blendOutTime=0.5}
    boundBone {boneName=Hair_03_l_big}
    componentAppliedLinearAccClamp=<500 500 500>
    componentLinearAccScale=<0.2 0.2 0.2>
    componentLinearVelScale=<0.3 0.3 0.3>
    gravityOverride {z=-980}
    physicsBodyDefinitions [{
      boundBone {boneName=Hair_03_l_big}
      boxExtents=<14 5 5>
      constraintSetup {angularLimitsMax=<5 10 10>, angularLimitsMin=<-5 -5 -10>, angularTarget=<0 0 0>}
      localJointOffset=<1 0 0>
     }]
   }
   n16 = Anim Dynamics - Bone: Hair_03_mid!(ComponentPose=@n15, bAlphaBoolEnabled:bool=True, ExternalForce="0,0,0") -> (Pose) {
    alphaInputType=Bool
    angularDampingOverride=0.9999
    bOverrideAngularDamping=true
    bUseGravityOverride=true
    simulationSpace=RootRelative
    alphaBoolBlend {blendInTime=0.1, blendOption=Linear, blendOutTime=0.2}
    boundBone {boneName=Hair_03_mid}
    componentAppliedLinearAccClamp=<1500 1500 1500>
    componentLinearAccScale=<0.2 0.2 0.2>
    componentLinearVelScale=<0.3 0.3 0.3>
    gravityOverride {z=-980}
    physicsBodyDefinitions [{
      boundBone {boneName=Hair_03_mid}
      boxExtents=<10 5 5>
      constraintSetup {angularLimitsMax=<5 15 10>, angularLimitsMin=<-5 -5 5>, angularTarget=<0 0 0>}
      localJointOffset=<1 0 0>
     }]
   }
   n17 = Anim Dynamics - Bone: Hair_01_mid!(ComponentPose=@n16, bAlphaBoolEnabled:bool=True, ExternalForce="0,0,0") -> (Pose) {
    alphaInputType=Bool
    angularDampingOverride=0.999
    bOverrideAngularDamping=true
    bUseGravityOverride=true
    simulationSpace=RootRelative
    alphaBoolBlend {blendInTime=0.1, blendOption=Linear, blendOutTime=0.2}
    boundBone {boneName=Hair_01_mid}
    componentAppliedLinearAccClamp=<3000 3000 3000>
    componentLinearAccScale=<0.2 0.2 0.2>
    componentLinearVelScale=<0.3 0.3 0.3>
    gravityOverride {z=-980}
    physicsBodyDefinitions [{
      boundBone {boneName=Hair_01_mid}
      boxExtents=<12 7 7>
      constraintSetup {angularLimitsMax=<1 15 10>, angularLimitsMin=<-1 -5 -10>, angularTarget=<0 0 0>}
      localJointOffset=<1 0 0>
     }]
   }
   n18 = Anim Dynamics - Bone: Hair_02_mid!(ComponentPose=@n17, bAlphaBoolEnabled:bool=True, ExternalForce="0,0,0") -> (Pose) {
    alphaInputType=Bool
    angularDampingOverride=0.9999
    bOverrideAngularDamping=true
    bUseGravityOverride=true
    simulationSpace=RootRelative
    alphaBoolBlend {blendInTime=0.1, blendOption=Linear, blendOutTime=0.2}
    boundBone {boneName=Hair_02_mid}
    componentAppliedLinearAccClamp=<1500 1500 1500>
    componentLinearAccScale=<0.2 0.2 0.2>
    componentLinearVelScale=<0.3 0.3 0.3>
    gravityOverride {z=-980}
    physicsBodyDefinitions [{
      boundBone {boneName=Hair_02_mid}
      boxExtents=<10 5 5>
      constraintSetup {angularLimitsMax=<1 15 10>, angularLimitsMin=<-1 -5 -10>, angularTarget=<0 0 0>}
      localJointOffset=<1 0 0>
     }]
   }
   n19 = Anim Dynamics - Bone: dyn_bundle!(ComponentPose=@n18, bAlphaBoolEnabled:bool=True, ExternalForce="0,0,0") -> (Pose) {
    alphaInputType=Bool
    angularDampingOverride=0.9999
    angularSpringConstant=4500
    bAngularSpring=true
    bOverrideAngularDamping=true
    bUseGravityOverride=true
    simulationSpace=RootRelative
    alphaBoolBlend {blendInTime=0.1, blendOption=Linear, blendOutTime=0.2}
    boundBone {boneName=dyn_bundle}
    componentAppliedLinearAccClamp=<1500 1500 1500>
    componentLinearAccScale=<0.2 0.2 0.2>
    componentLinearVelScale=<0.3 0.3 0.3>
    gravityOverride {z=-700}
    physicsBodyDefinitions [{
      boundBone {boneName=dyn_bundle}
      boxExtents=<9 5 5>
      constraintSetup {angularLimitsMax=<5 2 3>, angularLimitsMin=<-5 -20 -3>, angularTarget=<0 0 0>}
      localJointOffset=<1 0 0>
     }]
   }
   n20 = Anim Dynamics - Bone: dyn_lok_01!(ComponentPose=@n19, bAlphaBoolEnabled:bool=True, ExternalForce="0,0,0") -> (Pose) {
    alphaInputType=Bool
    angularDampingOverride=0.999
    bOverrideAngularDamping=true
    bUseGravityOverride=true
    simulationSpace=RootRelative
    alphaBoolBlend {blendInTime=0.1, blendOption=Linear, blendOutTime=0.2}
    boundBone {boneName=dyn_lok_01}
    componentAppliedLinearAccClamp=<3000 3000 3000>
    componentLinearAccScale=<0.2 0.2 0.2>
    componentLinearVelScale=<0.3 0.3 0.3>
    gravityOverride {z=-980}
    physicsBodyDefinitions [{
      boundBone {boneName=dyn_lok_01}
      boxExtents=<12 7 7>
      constraintSetup {angularLimitsMax=<5 15 5>, angularLimitsMin=<-5 -5 -5>, angularTarget=<0 0 0>}
      localJointOffset=<1 0 0>
     }]
   }
   n21 = RigidBody!(ComponentPose=@n20, Alpha=0.8, OverrideWorldGravity:Vector="0,0,0", ExternalForce) -> (Pose) {
    bOverrideWorldGravity=true
    overridePhysicsAsset=&/F_MED_BiscuitFluff_FaceAcc_Physics
    alphaBoolBlend {blendOption=Linear}
    componentAppliedLinearAccClamp=<5000 5000 5000>
    componentLinearAccScale {y=0.5, z=0.5}
    componentLinearVelScale {y=0.5, z=0.5}
    simSpaceSettings {
     maxAngularAcceleration=500
     maxAngularVelocity=3
     maxLinearAcceleration=400
     maxLinearVelocity=250
     velocityScaleZ=0
     worldAlpha=1
     externalLinearDragV=<0.1 0.1>
    }
   }
   Output Pose(Result=Component To Local(ComponentPose=@n21)) {layerGroup=None}
   Reroute Node(InputPin:Wildcard) -> (OutputPin:Wildcard)
  ]}
 EventGraph {nodes [
   Event BlueprintUpdateAnimation()
   > n1 = Gravity Override State Function DEPRECATED(CharacterPartAnimInstance=Self-Reference(), GameTimeInSeconds=GetGameTimeInSeconds(), GravityOverrideParams=Get Gravity Override Params())
   > Set Gravity Override(Gravity Override=@n1)
  ]}
]
variables [
 Gravity Override {type=Vector}
 Gravity Override Params {type=GravityOverrideParamsStruct}
]
